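# Physics asset for the F_MED_Musician female player body cosmetic, simulating a skirt around pelvis and thigh capsules.
# Twelve skirt chains each link three box bodies to a small capsule MSTR root, using uniform swing and narrow twist limits.
# The four off-center front roots attach to their side's thigh and the other eight to the pelvis.
# These root joints and the thigh-to-pelvis joints have free swing and twist.
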PhysicsAsset /BRCosmetics/Characters/Player/Female/Medium/Bodies/F_MED_Musician/Meshes/F_MED_Musician_Physics
bodies {
 dyn_skirt_01_l [{shapeType=Box, extentX=9.6316, extentY=13.2247, extentZ=1.5, center=<4.9784 0 0>}]
 dyn_skirt_01_l_MSTR [{shapeType=Capsule, length=1, radius=1}]
 dyn_skirt_01_r [{shapeType=Box, extentX=9.6316, extentY=13.2247, extentZ=1.5, center=<4.9784 0 0>}]
 dyn_skirt_01_r_MSTR [{shapeType=Capsule, length=1, radius=1}]
 dyn_skirt_02_l [{shapeType=Box, extentX=9.6316, extentY=13.2247, extentZ=1.5, center=<4.9784 0 0>}]
 dyn_skirt_02_r [{shapeType=Box, extentX=9.6316, extentY=13.2247, extentZ=1.5, center=<4.9784 0 0>}]
 dyn_skirt_03_l [{shapeType=Box, extentX=9.6316, extentY=13.2247, extentZ=1.5, center=<4.9784 0 0>}]
 dyn_skirt_03_r [{shapeType=Box, extentX=9.6316, extentY=13.2247, extentZ=1.5, center=<4.9784 0 0>}]
 dyn_skirt_bk2_01_l [{shapeType=Box, extentX=9.6316, extentY=13.2247, extentZ=1.5, center=<4.9784 0 0>}]
 dyn_skirt_bk2_01_l_MSTR [{shapeType=Capsule, length=1, radius=1}]
 dyn_skirt_bk2_01_r [{shapeType=Box, extentX=9.6316, extentY=13.2247, extentZ=1.5, center=<4.9784 0 0>}]
 dyn_skirt_bk2_01_r_MSTR [{shapeType=Capsule, length=1, radius=1}]
 dyn_skirt_bk2_02_l [{shapeType=Box, extentX=9.6316, extentY=13.2247, extentZ=1.5, center=<4.9784 0 0>}]
 dyn_skirt_bk2_02_r [{shapeType=Box, extentX=9.6316, extentY=13.2247, extentZ=1.5, center=<4.9784 0 0>}]
 dyn_skirt_bk2_03_l [{shapeType=Box, extentX=9.6316, extentY=13.2247, extentZ=1.5, center=<4.9784 0 0>}]
 dyn_skirt_bk2_03_r [{shapeType=Box, extentX=9.6316, extentY=13.2247, extentZ=1.5, center=<4.9784 0 0>}]
 dyn_skirt_bk_01 [{shapeType=Box, extentX=9.6316, extentY=13.2247, extentZ=1.5, center=<4.9784 0 0>}]
 dyn_skirt_bk_01_MSTR [{shapeType=Capsule, length=1, radius=1}]
 dyn_skirt_bk_01_l [{shapeType=Box, extentX=9.6316, extentY=13.2247, extentZ=1.5, center=<4.9784 0 0>}]
 dyn_skirt_bk_01_l_MSTR [{shapeType=Capsule, length=1, radius=1}]
 dyn_skirt_bk_01_r [{shapeType=Box, extentX=9.6316, extentY=13.2247, extentZ=1.5, center=<4.9784 0 0>}]
 dyn_skirt_bk_01_r_MSTR [{shapeType=Capsule, length=1, radius=1}]
 dyn_skirt_bk_02 [{shapeType=Box, extentX=9.6316, extentY=13.2247, extentZ=3.822, center=<4.9784 0 0>}]
 dyn_skirt_bk_02_l [{shapeType=Box, extentX=9.6316, extentY=13.2247, extentZ=3.822, center=<4.9784 0 0>}]
 dyn_skirt_bk_02_r [{shapeType=Box, extentX=9.6316, extentY=13.2247, extentZ=3.822, center=<4.9784 0 0>}]
 dyn_skirt_bk_03 [{shapeType=Box, extentX=9.6316, extentY=13.2247, extentZ=3.822, center=<4.9784 0 0>}]
 dyn_skirt_bk_03_l [{shapeType=Box, extentX=9.6316, extentY=13.2247, extentZ=3.822, center=<4.9784 0 0>}]
 dyn_skirt_bk_03_r [{shapeType=Box, extentX=9.6316, extentY=13.2247, extentZ=3.822, center=<4.9784 0 0>}]
 dyn_skirt_fr2_01_l [{shapeType=Box, extentX=9.6316, extentY=13.2247, extentZ=1.5, center=<4.9784 0 0>}]
 dyn_skirt_fr2_01_l_MSTR [{shapeType=Capsule, length=1, radius=1}]
 dyn_skirt_fr2_01_r [{shapeType=Box, extentX=9.6316, extentY=13.2247, extentZ=1.5, center=<4.9784 0 0>}]
 dyn_skirt_fr2_01_r_MSTR [{shapeType=Capsule, length=1, radius=1}]
 dyn_skirt_fr2_02_l [{shapeType=Box, extentX=9.6316, extentY=13.2247, extentZ=1.5, center=<4.9784 0 0>}]
 dyn_skirt_fr2_02_r [{shapeType=Box, extentX=9.6316, extentY=13.2247, extentZ=1.5, center=<4.9784 0 0>}]
 dyn_skirt_fr2_03_l [{shapeType=Box, extentX=9.6316, extentY=13.2247, extentZ=1.5, center=<4.9784 0 0>}]
 dyn_skirt_fr2_03_r [{shapeType=Box, extentX=9.6316, extentY=13.2247, extentZ=1.5, center=<4.9784 0 0>}]
 dyn_skirt_fr_01 [{shapeType=Box, extentX=9.6316, extentY=13.2247, extentZ=1.5, center=<4.9784 0 0>}]
 dyn_skirt_fr_01_MSTR [{shapeType=Capsule, length=1, radius=1}]
 dyn_skirt_fr_01_l [{shapeType=Box, extentX=9.6316, extentY=13.2247, extentZ=1.5, center=<4.9784 0 0>}]
 dyn_skirt_fr_01_l_MSTR [{shapeType=Capsule, length=1, radius=1}]
 dyn_skirt_fr_01_r [{shapeType=Box, extentX=9.6316, extentY=13.2247, extentZ=1.5, center=<4.9784 0 0>}]
 dyn_skirt_fr_01_r_MSTR [{shapeType=Capsule, length=1, radius=1}]
 dyn_skirt_fr_02 [{shapeType=Box, extentX=9.6316, extentY=13.2247, extentZ=1.5, center=<4.9784 0 0>}]
 dyn_skirt_fr_02_l [{shapeType=Box, extentX=9.6316, extentY=13.2247, extentZ=1.5, center=<4.9784 0 0>}]
 dyn_skirt_fr_02_r [{shapeType=Box, extentX=9.6316, extentY=13.2247, extentZ=1.5, center=<4.9784 0 0>}]
 dyn_skirt_fr_03 [{shapeType=Box, extentX=9.6316, extentY=13.2247, extentZ=1.5, center=<4.9784 0 0>}]
 dyn_skirt_fr_03_l [{shapeType=Box, extentX=9.6316, extentY=13.2247, extentZ=1.5, center=<4.9784 0 0>}]
 dyn_skirt_fr_03_r [{shapeType=Box, extentX=9.6316, extentY=13.2247, extentZ=1.5, center=<4.9784 0 0>}]
 pelvis [{shapeType=Capsule, length=7.458, radius=8.0636, center=<1.1549 -1.274 0>}]
 thigh_l [{shapeType=Capsule, length=28.5266, radius=8.8056, center=<-21.4384 -0.2237 0>, rotation=<rot 90 0 0>}]
 thigh_r [{shapeType=Capsule, length=29.0731, radius=8.2591, center=<21.3234 0.3071 0>, rotation=<rot 90 0 0>}]
}
constraints [
 {
  bone1Name=thigh_l
  bone2Name=pelvis
  swing1LimitDegrees=45
  swing1Motion=Free
  swing2LimitDegrees=45
  swing2Motion=Free
  twistLimitDegrees=45
  twistMotion=Free
 }
 {
  bone1Name=thigh_r
  bone2Name=pelvis
  swing1LimitDegrees=45
  swing1Motion=Free
  swing2LimitDegrees=45
  swing2Motion=Free
  twistLimitDegrees=45
  twistMotion=Free
 }
 {
  bone1Name=dyn_skirt_fr_01_l_MSTR
  bone2Name=thigh_l
  swing1LimitDegrees=45
  swing1Motion=Free
  swing2LimitDegrees=45
  swing2Motion=Free
  twistLimitDegrees=45
  twistMotion=Free
 }
 {
  bone1Name=dyn_skirt_fr2_01_l_MSTR
  bone2Name=thigh_l
  swing1LimitDegrees=45
  swing1Motion=Free
  swing2LimitDegrees=45
  swing2Motion=Free
  twistLimitDegrees=45
  twistMotion=Free
 }
 {
  bone1Name=dyn_skirt_fr_01_r_MSTR
  bone2Name=thigh_r
  swing1LimitDegrees=45
  swing1Motion=Free
  swing2LimitDegrees=45
  swing2Motion=Free
  twistLimitDegrees=45
  twistMotion=Free
 }
 {
  bone1Name=dyn_skirt_fr2_01_r_MSTR
  bone2Name=thigh_r
  swing1LimitDegrees=45
  swing1Motion=Free
  swing2LimitDegrees=45
  swing2Motion=Free
  twistLimitDegrees=45
  twistMotion=Free
 }
 {
  bone1Name=dyn_skirt_fr_01_MSTR
  bone2Name=pelvis
  swing1LimitDegrees=45
  swing1Motion=Free
  swing2LimitDegrees=45
  swing2Motion=Free
  twistLimitDegrees=45
  twistMotion=Free
 }
 {
  bone1Name=dyn_skirt_bk_01_MSTR
  bone2Name=pelvis
  swing1LimitDegrees=45
  swing1Motion=Free
  swing2LimitDegrees=45
  swing2Motion=Free
  twistLimitDegrees=45
  twistMotion=Free
 }
 {
  bone1Name=dyn_skirt_01_r_MSTR
  bone2Name=pelvis
  swing1LimitDegrees=45
  swing1Motion=Free
  swing2LimitDegrees=45
  swing2Motion=Free
  twistLimitDegrees=45
  twistMotion=Free
 }
 {
  bone1Name=dyn_skirt_bk2_01_r_MSTR
  bone2Name=pelvis
  swing1LimitDegrees=45
  swing1Motion=Free
  swing2LimitDegrees=45
  swing2Motion=Free
  twistLimitDegrees=45
  twistMotion=Free
 }
 {
  bone1Name=dyn_skirt_bk_01_r_MSTR
  bone2Name=pelvis
  swing1LimitDegrees=45
  swing1Motion=Free
  swing2LimitDegrees=45
  swing2Motion=Free
  twistLimitDegrees=45
  twistMotion=Free
 }
 {
  bone1Name=dyn_skirt_bk_01_l_MSTR
  bone2Name=pelvis
  swing1LimitDegrees=45
  swing1Motion=Free
  swing2LimitDegrees=45
  swing2Motion=Free
  twistLimitDegrees=45
  twistMotion=Free
 }
 {
  bone1Name=dyn_skirt_01_l_MSTR
  bone2Name=pelvis
  swing1LimitDegrees=45
  swing1Motion=Free
  swing2LimitDegrees=45
  swing2Motion=Free
  twistLimitDegrees=45
  twistMotion=Free
 }
 {
  bone1Name=dyn_skirt_bk2_01_l_MSTR
  bone2Name=pelvis
  swing1LimitDegrees=45
  swing1Motion=Free
  swing2LimitDegrees=45
  swing2Motion=Free
  twistLimitDegrees=45
  twistMotion=Free
 }
 {bone1Name=dyn_skirt_03_l, bone2Name=dyn_skirt_02_l, swing1LimitDegrees=25, swing2LimitDegrees=45, twistLimitDegrees=5}
 {bone1Name=dyn_skirt_01_l, bone2Name=dyn_skirt_01_l_MSTR, swing1LimitDegrees=25, swing2LimitDegrees=45, twistLimitDegrees=5}
 {bone1Name=dyn_skirt_02_l, bone2Name=dyn_skirt_01_l, swing1LimitDegrees=25, swing2LimitDegrees=45, twistLimitDegrees=5}
 {bone1Name=dyn_skirt_bk2_03_l, bone2Name=dyn_skirt_bk2_02_l, swing1LimitDegrees=25, swing2LimitDegrees=45, twistLimitDegrees=5}
 {bone1Name=dyn_skirt_bk2_01_l, bone2Name=dyn_skirt_bk2_01_l_MSTR, swing1LimitDegrees=25, swing2LimitDegrees=45, twistLimitDegrees=5}
 {bone1Name=dyn_skirt_bk2_02_l, bone2Name=dyn_skirt_bk2_01_l, swing1LimitDegrees=25, swing2LimitDegrees=45, twistLimitDegrees=5}
 {bone1Name=dyn_skirt_bk_03_l, bone2Name=dyn_skirt_bk_02_l, swing1LimitDegrees=25, swing2LimitDegrees=45, twistLimitDegrees=5}
 {bone1Name=dyn_skirt_bk_01_l, bone2Name=dyn_skirt_bk_01_l_MSTR, swing1LimitDegrees=25, swing2LimitDegrees=45, twistLimitDegrees=5}
 {bone1Name=dyn_skirt_bk_02_l, bone2Name=dyn_skirt_bk_01_l, swing1LimitDegrees=25, swing2LimitDegrees=45, twistLimitDegrees=5}
 {bone1Name=dyn_skirt_bk_03_r, bone2Name=dyn_skirt_bk_02_r, swing1LimitDegrees=25, swing2LimitDegrees=45, twistLimitDegrees=5}
 {bone1Name=dyn_skirt_bk_01_r, bone2Name=dyn_skirt_bk_01_r_MSTR, swing1LimitDegrees=25, swing2LimitDegrees=45, twistLimitDegrees=5}
 {bone1Name=dyn_skirt_bk_02_r, bone2Name=dyn_skirt_bk_01_r, swing1LimitDegrees=25, swing2LimitDegrees=45, twistLimitDegrees=5}
 {bone1Name=dyn_skirt_bk2_03_r, bone2Name=dyn_skirt_bk2_02_r, swing1LimitDegrees=25, swing2LimitDegrees=45, twistLimitDegrees=5}
 {bone1Name=dyn_skirt_bk2_01_r, bone2Name=dyn_skirt_bk2_01_r_MSTR, swing1LimitDegrees=25, swing2LimitDegrees=45, twistLimitDegrees=5}
 {bone1Name=dyn_skirt_bk2_02_r, bone2Name=dyn_skirt_bk2_01_r, swing1LimitDegrees=25, swing2LimitDegrees=45, twistLimitDegrees=5}
 {bone1Name=dyn_skirt_03_r, bone2Name=dyn_skirt_02_r, swing1LimitDegrees=25, swing2LimitDegrees=45, twistLimitDegrees=5}
 {bone1Name=dyn_skirt_01_r, bone2Name=dyn_skirt_01_r_MSTR, swing1LimitDegrees=25, swing2LimitDegrees=45, twistLimitDegrees=5}
 {bone1Name=dyn_skirt_02_r, bone2Name=dyn_skirt_01_r, swing1LimitDegrees=25, swing2LimitDegrees=45, twistLimitDegrees=5}
 {bone1Name=dyn_skirt_bk_03, bone2Name=dyn_skirt_bk_02, swing1LimitDegrees=25, swing2LimitDegrees=45, twistLimitDegrees=5}
 {bone1Name=dyn_skirt_bk_01, bone2Name=dyn_skirt_bk_01_MSTR, swing1LimitDegrees=25, swing2LimitDegrees=45, twistLimitDegrees=5}
 {bone1Name=dyn_skirt_bk_02, bone2Name=dyn_skirt_bk_01, swing1LimitDegrees=25, swing2LimitDegrees=45, twistLimitDegrees=5}
 {bone1Name=dyn_skirt_fr_03, bone2Name=dyn_skirt_fr_02, swing1LimitDegrees=25, swing2LimitDegrees=45, twistLimitDegrees=5}
 {bone1Name=dyn_skirt_fr_01, bone2Name=dyn_skirt_fr_01_MSTR, swing1LimitDegrees=25, swing2LimitDegrees=45, twistLimitDegrees=5}
 {bone1Name=dyn_skirt_fr_02, bone2Name=dyn_skirt_fr_01, swing1LimitDegrees=25, swing2LimitDegrees=45, twistLimitDegrees=5}
 {bone1Name=dyn_skirt_fr2_03_r, bone2Name=dyn_skirt_fr2_02_r, swing1LimitDegrees=25, swing2LimitDegrees=45, twistLimitDegrees=5}
 {bone1Name=dyn_skirt_fr2_01_r, bone2Name=dyn_skirt_fr2_01_r_MSTR, swing1LimitDegrees=25, swing2LimitDegrees=45, twistLimitDegrees=5}
 {bone1Name=dyn_skirt_fr2_02_r, bone2Name=dyn_skirt_fr2_01_r, swing1LimitDegrees=25, swing2LimitDegrees=45, twistLimitDegrees=5}
 {bone1Name=dyn_skirt_fr_03_r, bone2Name=dyn_skirt_fr_02_r, swing1LimitDegrees=25, swing2LimitDegrees=45, twistLimitDegrees=5}
 {bone1Name=dyn_skirt_fr_01_r, bone2Name=dyn_skirt_fr_01_r_MSTR, swing1LimitDegrees=25, swing2LimitDegrees=45, twistLimitDegrees=5}
 {bone1Name=dyn_skirt_fr_02_r, bone2Name=dyn_skirt_fr_01_r, swing1LimitDegrees=25, swing2LimitDegrees=45, twistLimitDegrees=5}
 {bone1Name=dyn_skirt_fr2_03_l, bone2Name=dyn_skirt_fr2_02_l, swing1LimitDegrees=25, swing2LimitDegrees=45, twistLimitDegrees=5}
 {bone1Name=dyn_skirt_fr2_01_l, bone2Name=dyn_skirt_fr2_01_l_MSTR, swing1LimitDegrees=25, swing2LimitDegrees=45, twistLimitDegrees=5}
 {bone1Name=dyn_skirt_fr2_02_l, bone2Name=dyn_skirt_fr2_01_l, swing1LimitDegrees=25, swing2LimitDegrees=45, twistLimitDegrees=5}
 {bone1Name=dyn_skirt_fr_03_l, bone2Name=dyn_skirt_fr_02_l, swing1LimitDegrees=25, swing2LimitDegrees=45, twistLimitDegrees=5}
 {bone1Name=dyn_skirt_fr_01_l, bone2Name=dyn_skirt_fr_01_l_MSTR, swing1LimitDegrees=25, swing2LimitDegrees=45, twistLimitDegrees=5}
 {bone1Name=dyn_skirt_fr_02_l, bone2Name=dyn_skirt_fr_01_l, swing1LimitDegrees=25, swing2LimitDegrees=45, twistLimitDegrees=5}
]
