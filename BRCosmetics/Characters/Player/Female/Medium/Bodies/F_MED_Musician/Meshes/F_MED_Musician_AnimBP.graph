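# Animation Blueprint for the F_MED_Musician body that drives secondary motion on the skirt bones.
# From cached pose PreDressRBAN, skydiving or diving selects back-skirt constraints with AnimDynamics; otherwise a pelvis-space RigidBody simulation runs.
# Constraints then pull front bones toward the thighs and blend sides between front and back, before further AnimDynamics adds springy sway.
# Crouch-driven alpha selects shown return 1 either way.
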AnimBlueprint /BRCosmetics/Characters/Player/Female/Medium/Bodies/F_MED_Musician/Meshes/F_MED_Musician_AnimBP
target_skeleton=&/F_MED_Musician_Skeleton
graphs [
 AnimGraph {nodes [
   Component To Local()
   n1 = "Constraint - dyn_skirt_bk_01 (RPP)"!(
    ComponentPose=Local To Component(LocalPose=Use cached pose 'PreDressRBAN'())
    Alpha=1
    ConstraintWeights_0=0
    ConstraintWeights_1:float=0.5
    ConstraintWeights_2:float=0.5
   ) -> (Pose) {
    alphaBoolBlend {blendOption=Linear}
    boneToModify {boneName=dyn_skirt_bk_01}
    constraintSetup [{transformType=Rotation, targetBone {boneName=pelvis}}, {targetBone {boneName=dyn_skirt_bk_01_l}}, {targetBone {boneName=dyn_skirt_bk_01_r}}]
    constraintWeights [1,1,1]
   }
   n2 = "Constraint - dyn_skirt_bk_02 (RPP)"!(ComponentPose=@n1, Alpha=1, ConstraintWeights_0=0, ConstraintWeights_1:float=0.5, ConstraintWeights_2:float=0.5) -> (Pose) {
    alphaBoolBlend {blendOption=Linear}
    boneToModify {boneName=dyn_skirt_bk_02}
    constraintSetup [{transformType=Rotation, targetBone {boneName=pelvis}}, {targetBone {boneName=dyn_skirt_bk_02_l}}, {targetBone {boneName=dyn_skirt_bk_02_r}}]
    constraintWeights [1,1,1]
   }
   n3 = "Constraint - dyn_skirt_bk_03 (RPP)"!(ComponentPose=@n2, Alpha=1, ConstraintWeights_0=0, ConstraintWeights_1:float=0.5, ConstraintWeights_2:float=0.5) -> (Pose) {
    alphaBoolBlend {blendOption=Linear}
    boneToModify {boneName=dyn_skirt_bk_03}
    constraintSetup [{transformType=Rotation, targetBone {boneName=pelvis}}, {targetBone {boneName=dyn_skirt_bk_03_l}}, {targetBone {boneName=dyn_skirt_bk_03_r}}]
    constraintWeights [1,1,1]
   }
   n4 = "Constraint - dyn_skirt_bk2_01_l (P)"(ComponentPose=@n3, Alpha=1, ConstraintWeights_0=1) {
    alphaBoolBlend {blendOption=Linear}
    boneToModify {boneName=dyn_skirt_bk2_01_l}
    constraintSetup [{targetBone {boneName=pelvis}}]
    constraintWeights [1]
   }
   n5 = "Constraint - dyn_skirt_bk2_02_l (P)"(ComponentPose=@n4, Alpha=0.3, ConstraintWeights_0=1) {
    alphaBoolBlend {blendOption=Linear}
    boneToModify {boneName=dyn_skirt_bk2_02_l}
    constraintSetup [{targetBone {boneName=pelvis}}]
    constraintWeights [1]
   }
   n6 = "Constraint - dyn_skirt_bk2_03_l (P)"(ComponentPose=@n5, Alpha=1, ConstraintWeights_0=1) {
    alphaBoolBlend {blendOption=Linear}
    boneToModify {boneName=dyn_skirt_bk2_03_l}
    constraintSetup [{targetBone {boneName=pelvis}}]
    constraintWeights [1]
   }
   n7 = "Constraint - dyn_skirt_bk_01_l (RRR)"!(ComponentPose=@n6, Alpha=1, ConstraintWeights_0=0, ConstraintWeights_1:float=0.5, ConstraintWeights_2:float=0.5) -> (Pose) {
    alphaBoolBlend {blendOption=Linear}
    boneToModify {boneName=dyn_skirt_bk_01_l}
    constraintSetup [
     {transformType=Rotation, targetBone {boneName=pelvis}}
     {transformType=Rotation, targetBone {boneName=dyn_skirt_bk_01}}
     {transformType=Rotation, targetBone {boneName=dyn_skirt_bk2_01_l}}
    ]
    constraintWeights [1,1,1]
   }
   n8 = "Constraint - dyn_skirt_bk_02_l (RRR)"!(ComponentPose=@n7, Alpha=1, ConstraintWeights_0=0, ConstraintWeights_1:float=0.5, ConstraintWeights_2:float=0.5) -> (Pose) {
    alphaBoolBlend {blendOption=Linear}
    boneToModify {boneName=dyn_skirt_bk_02_l}
    constraintSetup [
     {transformType=Rotation, targetBone {boneName=pelvis}}
     {transformType=Rotation, targetBone {boneName=dyn_skirt_bk_02}}
     {transformType=Rotation, targetBone {boneName=dyn_skirt_bk2_02_l}}
    ]
    constraintWeights [1,1,1]
   }
   n9 = "Constraint - dyn_skirt_bk_03_l (RRR)"!(ComponentPose=@n8, Alpha=1, ConstraintWeights_0=0, ConstraintWeights_1:float=0.5, ConstraintWeights_2:float=0.5) -> (Pose) {
    alphaBoolBlend {blendOption=Linear}
    boneToModify {boneName=dyn_skirt_bk_03_l}
    constraintSetup [
     {transformType=Rotation, targetBone {boneName=pelvis}}
     {transformType=Rotation, targetBone {boneName=dyn_skirt_bk_03}}
     {transformType=Rotation, targetBone {boneName=dyn_skirt_bk2_03_l}}
    ]
    constraintWeights [1,1,1]
   }
   n10 = "Constraint - dyn_skirt_bk_01_r (RRR)"!(ComponentPose=@n9, Alpha=1, ConstraintWeights_0=0, ConstraintWeights_1:float=0.5, ConstraintWeights_2:float=0.5) -> (Pose) {
    alphaBoolBlend {blendOption=Linear}
    boneToModify {boneName=dyn_skirt_bk_01_r}
    constraintSetup [
     {transformType=Rotation, targetBone {boneName=pelvis}}
     {transformType=Rotation, targetBone {boneName=dyn_skirt_bk_01}}
     {transformType=Rotation, targetBone {boneName=dyn_skirt_bk2_01_r}}
    ]
    constraintWeights [1,1,1]
   }
   n11 = "Constraint - dyn_skirt_bk_02_r (RRR)"!(ComponentPose=@n10, Alpha=1, ConstraintWeights_0=0, ConstraintWeights_1:float=0.5, ConstraintWeights_2:float=0.5) -> (Pose) {
    alphaBoolBlend {blendOption=Linear}
    boneToModify {boneName=dyn_skirt_bk_02_r}
    constraintSetup [
     {transformType=Rotation, targetBone {boneName=pelvis}}
     {transformType=Rotation, targetBone {boneName=dyn_skirt_bk_02}}
     {transformType=Rotation, targetBone {boneName=dyn_skirt_bk2_02_r}}
    ]
    constraintWeights [1,1,1]
   }
   n12 = "Constraint - dyn_skirt_bk_03_r (RRR)"!(ComponentPose=@n11, Alpha=1, ConstraintWeights_0=0, ConstraintWeights_1:float=0.5, ConstraintWeights_2:float=0.5) -> (Pose) {
    alphaBoolBlend {blendOption=Linear}
    boneToModify {boneName=dyn_skirt_bk_03_r}
    constraintSetup [
     {transformType=Rotation, targetBone {boneName=pelvis}}
     {transformType=Rotation, targetBone {boneName=dyn_skirt_bk_03}}
     {transformType=Rotation, targetBone {boneName=dyn_skirt_bk2_03_r}}
    ]
    constraintWeights [1,1,1]
   }
   n13 = "Constraint - dyn_skirt_bk2_01_r (P)"(ComponentPose=@n12, Alpha=1, ConstraintWeights_0=1) {
    alphaBoolBlend {blendOption=Linear}
    boneToModify {boneName=dyn_skirt_bk2_01_r}
    constraintSetup [{targetBone {boneName=pelvis}}]
    constraintWeights [1]
   }
   n14 = "Constraint - dyn_skirt_bk2_02_r (P)"(ComponentPose=@n13, Alpha=1, ConstraintWeights_0=1) {
    alphaBoolBlend {blendOption=Linear}
    boneToModify {boneName=dyn_skirt_bk2_02_r}
    constraintSetup [{targetBone {boneName=pelvis}}]
    constraintWeights [1]
   }
   n15 = "Constraint - dyn_skirt_bk2_03_r (P)"(ComponentPose=@n14, Alpha=1, ConstraintWeights_0=1) {
    alphaBoolBlend {blendOption=Linear}
    boneToModify {boneName=dyn_skirt_bk2_03_r}
    constraintSetup [{targetBone {boneName=pelvis}}]
    constraintWeights [1]
   }
   n16 = Anim Dynamics - Bone: dyn_skirt_bk_01_l(ComponentPose=@n15, Alpha=0.5, ExternalForce="0,0,0") {
    angularDampingOverride=0.9999
    angularSpringConstant=75
    bAngularSpring=true
    bGravityOverrideInSimSpace=true
    bOverrideAngularDamping=true
    simulationSpace=RootRelative
    alphaBoolBlend {blendOption=Linear}
    boundBone {boneName=dyn_skirt_bk_01_l}
    componentLinearAccScale=<1.2 1.2 1.2>
    componentLinearVelScale=<1.2 1.2 1.2>
    physicsBodyDefinitions [{
      sphereCollisionRadius=0
      boundBone {boneName=dyn_skirt_bk_01_l}
      boxExtents=<20 20 10>
      constraintSetup {angularLimitsMax=<25 0 25>, angularLimitsMin=<-25 -25 -25>, angularTarget=<0 0 0>}
      localJointOffset=<2 0 0>
     }]
   }
   n17 = Anim Dynamics - Bone: dyn_skirt_bk_02_l(ComponentPose=@n16, Alpha=0.5, ExternalForce="0,0,0") {
    angularDampingOverride=0.9999
    angularSpringConstant=75
    bAngularSpring=true
    bGravityOverrideInSimSpace=true
    bOverrideAngularDamping=true
    simulationSpace=RootRelative
    alphaBoolBlend {blendOption=Linear}
    boundBone {boneName=dyn_skirt_bk_02_l}
    componentLinearAccScale=<1.2 1.2 1.2>
    componentLinearVelScale=<1.2 1.2 1.2>
    physicsBodyDefinitions [{
      sphereCollisionRadius=0
      boundBone {boneName=dyn_skirt_bk_02_l}
      boxExtents=<20 20 10>
      constraintSetup {angularLimitsMax=<25 0 25>, angularLimitsMin=<-25 -25 -25>, angularTarget=<0 0 0>}
      localJointOffset=<2 0 0>
     }]
   }
   n18 = Anim Dynamics - Bone: dyn_skirt_bk_03_l(ComponentPose=@n17, Alpha=0.5, ExternalForce="0,0,0") {
    angularDampingOverride=0.9999
    angularSpringConstant=75
    bAngularSpring=true
    bGravityOverrideInSimSpace=true
    bOverrideAngularDamping=true
    simulationSpace=RootRelative
    alphaBoolBlend {blendOption=Linear}
    boundBone {boneName=dyn_skirt_bk_03_l}
    componentLinearAccScale=<1.2 1.2 1.2>
    componentLinearVelScale=<1.2 1.2 1.2>
    physicsBodyDefinitions [{
      sphereCollisionRadius=0
      boundBone {boneName=dyn_skirt_bk_03_l}
      boxExtents=<20 20 10>
      constraintSetup {angularLimitsMax=<25 0 25>, angularLimitsMin=<-25 -25 -25>, angularTarget=<0 0 0>}
      localJointOffset=<2 0 0>
     }]
   }
   n19 = Anim Dynamics - Bone: dyn_skirt_bk_01(ComponentPose=@n18, Alpha=0.5, ExternalForce="0,0,0") {
    angularDampingOverride=0.9999
    angularSpringConstant=75
    bAngularSpring=true
    bGravityOverrideInSimSpace=true
    bOverrideAngularDamping=true
    simulationSpace=RootRelative
    alphaBoolBlend {blendOption=Linear}
    boundBone {boneName=dyn_skirt_bk_01}
    componentLinearAccScale=<1.2 1.2 1.2>
    componentLinearVelScale=<1.2 1.2 1.2>
    physicsBodyDefinitions [{
      sphereCollisionRadius=0
      boundBone {boneName=dyn_skirt_bk_01}
      boxExtents=<20 20 10>
      constraintSetup {angularLimitsMax=<25 0 25>, angularLimitsMin=<-25 -25 -25>, angularTarget=<0 0 0>}
      localJointOffset=<2 0 0>
     }]
   }
   n20 = Anim Dynamics - Bone: dyn_skirt_bk_02(ComponentPose=@n19, Alpha=0.5, ExternalForce="0,0,0") {
    angularDampingOverride=0.9999
    angularSpringConstant=75
    bAngularSpring=true
    bGravityOverrideInSimSpace=true
    bOverrideAngularDamping=true
    simulationSpace=RootRelative
    alphaBoolBlend {blendOption=Linear}
    boundBone {boneName=dyn_skirt_bk_02}
    componentLinearAccScale=<1.2 1.2 1.2>
    componentLinearVelScale=<1.2 1.2 1.2>
    physicsBodyDefinitions [{
      sphereCollisionRadius=0
      boundBone {boneName=dyn_skirt_bk_02}
      boxExtents=<20 20 10>
      constraintSetup {angularLimitsMax=<25 0 25>, angularLimitsMin=<-25 -25 -25>, angularTarget=<0 0 0>}
      localJointOffset=<2 0 0>
     }]
   }
   n21 = Anim Dynamics - Bone: dyn_skirt_bk_03(ComponentPose=@n20, Alpha=0.5, ExternalForce="0,0,0") {
    angularDampingOverride=0.9999
    angularSpringConstant=75
    bAngularSpring=true
    bGravityOverrideInSimSpace=true
    bOverrideAngularDamping=true
    simulationSpace=RootRelative
    alphaBoolBlend {blendOption=Linear}
    boundBone {boneName=dyn_skirt_bk_03}
    componentLinearAccScale=<1.2 1.2 1.2>
    componentLinearVelScale=<1.2 1.2 1.2>
    physicsBodyDefinitions [{
      sphereCollisionRadius=0
      boundBone {boneName=dyn_skirt_bk_03}
      boxExtents=<20 20 10>
      constraintSetup {angularLimitsMax=<25 0 25>, angularLimitsMin=<-25 -25 -25>, angularTarget=<0 0 0>}
      localJointOffset=<2 0 0>
     }]
   }
   n22 = Anim Dynamics - Bone: dyn_skirt_bk_01_r(ComponentPose=@n21, Alpha=0.5, ExternalForce="0,0,0") {
    angularDampingOverride=0.9999
    angularSpringConstant=75
    bAngularSpring=true
    bGravityOverrideInSimSpace=true
    bOverrideAngularDamping=true
    simulationSpace=RootRelative
    alphaBoolBlend {blendOption=Linear}
    boundBone {boneName=dyn_skirt_bk_01_r}
    componentLinearAccScale=<1.2 1.2 1.2>
    componentLinearVelScale=<1.2 1.2 1.2>
    physicsBodyDefinitions [{
      sphereCollisionRadius=0
      boundBone {boneName=dyn_skirt_bk_01_r}
      boxExtents=<20 20 10>
      constraintSetup {angularLimitsMax=<25 0 25>, angularLimitsMin=<-25 -25 -25>, angularTarget=<0 0 0>}
      localJointOffset=<2 0 0>
     }]
   }
   n23 = Anim Dynamics - Bone: dyn_skirt_bk_02_r(ComponentPose=@n22, Alpha=0.5, ExternalForce="0,0,0") {
    angularDampingOverride=0.9999
    angularSpringConstant=75
    bAngularSpring=true
    bGravityOverrideInSimSpace=true
    bOverrideAngularDamping=true
    simulationSpace=RootRelative
    alphaBoolBlend {blendOption=Linear}
    boundBone {boneName=dyn_skirt_bk_02_r}
    componentLinearAccScale=<1.2 1.2 1.2>
    componentLinearVelScale=<1.2 1.2 1.2>
    physicsBodyDefinitions [{
      sphereCollisionRadius=0
      boundBone {boneName=dyn_skirt_bk_02_r}
      boxExtents=<20 20 10>
      constraintSetup {angularLimitsMax=<25 0 25>, angularLimitsMin=<-25 -25 -25>, angularTarget=<0 0 0>}
      localJointOffset=<2 0 0>
     }]
   }
   n24 = Anim Dynamics - Bone: dyn_skirt_bk_03_r(ComponentPose=@n23, Alpha=0.5, ExternalForce="0,0,0") {
    angularDampingOverride=0.9999
    angularSpringConstant=75
    bAngularSpring=true
    bGravityOverrideInSimSpace=true
    bOverrideAngularDamping=true
    simulationSpace=RootRelative
    alphaBoolBlend {blendOption=Linear}
    boundBone {boneName=dyn_skirt_bk_03_r}
    componentLinearAccScale=<1.2 1.2 1.2>
    componentLinearVelScale=<1.2 1.2 1.2>
    physicsBodyDefinitions [{
      sphereCollisionRadius=0
      boundBone {boneName=dyn_skirt_bk_03_r}
      boxExtents=<20 20 10>
      constraintSetup {angularLimitsMax=<25 0 25>, angularLimitsMin=<-25 -25 -25>, angularTarget=<0 0 0>}
      localJointOffset=<2 0 0>
     }]
   }
   n25 = Component To Local(ComponentPose=@n24)
   n26 = RigidBody!(
    ComponentPose=Local To Component(LocalPose=Use cached pose 'PreDressRBAN'())
    Alpha
    OverrideWorldGravity:Vector=Get Gravity Override()
    ExternalForce
   ) -> (Pose) {
    bClampLinearTranslationLimitToRefPose=true
    bOverrideWorldGravity=true
    overridePhysicsAsset=&/F_MED_Musician_Physics
    simulationSpace=BaseBoneSpace
    alphaBoolBlend {blendOption=Linear}
    baseBoneRef {boneName=pelvis}
    componentLinearAccScale=<0.28 0.28 0.28>
    overrideWorldGravity {z=-980}
   }
   n27 = Component To Local(ComponentPose=@n26)
   n28 = OR Boolean(A=Get bIsSkyDiving(), B=Get bIsSkydiveDiveMode())
   n29 = "Constraint - dyn_skirt_fr_01_l_MSTR (PP)"!(
    ComponentPose=Local To Component(LocalPose=Blend Poses by bool(BlendPose_0=@n25, BlendPose_1=@n27, BlendTime_0=0.01, BlendTime_1=0.01, bActiveValue=@n28) {blendType=Linear})
    Alpha=0.9
    ConstraintWeights_0=0
    ConstraintWeights_1:float=1
   ) -> (Pose) {
    alphaBoolBlend {blendOption=Linear}
    boneToModify {boneName=dyn_skirt_fr_01_l_MSTR}
    constraintSetup [{targetBone {boneName=pelvis}}, {targetBone {boneName=dyn_thigh_01_l}}]
    constraintWeights [1,1]
   }
   n30 = "Constraint - dyn_skirt_fr_01_l (PP)"!(ComponentPose=@n29, Alpha=0.9, ConstraintWeights_0=0, ConstraintWeights_1:float=1) -> (Pose) {
    alphaBoolBlend {blendOption=Linear}
    boneToModify {boneName=dyn_skirt_fr_01_l}
    constraintSetup [{targetBone {boneName=pelvis}}, {targetBone {boneName=dyn_thigh_01_l}}]
    constraintWeights [1,1]
   }
   n31 = "Constraint - dyn_skirt_fr_02_l (PP)"!(ComponentPose=@n30, Alpha=0.8, ConstraintWeights_0=0, ConstraintWeights_1:float=1) -> (Pose) {
    alphaBoolBlend {blendOption=Linear}
    boneToModify {boneName=dyn_skirt_fr_02_l}
    constraintSetup [{targetBone {boneName=pelvis}}, {targetBone {boneName=dyn_thigh_01_l}}]
    constraintWeights [1,1]
   }
   n32 = "Constraint - dyn_skirt_fr_03_l (PP)"!(ComponentPose=@n31, Alpha=0.5, ConstraintWeights_0=0, ConstraintWeights_1:float=1) -> (Pose) {
    alphaBoolBlend {blendOption=Linear}
    boneToModify {boneName=dyn_skirt_fr_03_l}
    constraintSetup [{targetBone {boneName=pelvis}}, {targetBone {boneName=dyn_thigh_01_l}}]
    constraintWeights [1,1]
   }
   n33 = "Constraint - dyn_skirt_fr2_01_l_MSTR (PP)"!(ComponentPose=@n32, Alpha=0.9, ConstraintWeights_0=0, ConstraintWeights_1:float=1) -> (Pose) {
    alphaBoolBlend {blendOption=Linear}
    boneToModify {boneName=dyn_skirt_fr2_01_l_MSTR}
    constraintSetup [{targetBone {boneName=pelvis}}, {targetBone {boneName=dyn_thigh_01_l}}]
    constraintWeights [1,1]
   }
   n34 = "Constraint - dyn_skirt_fr2_01_l (PP)"!(ComponentPose=@n33, Alpha=0.9, ConstraintWeights_0=0, ConstraintWeights_1:float=1) -> (Pose) {
    alphaBoolBlend {blendOption=Linear}
    boneToModify {boneName=dyn_skirt_fr2_01_l}
    constraintSetup [{targetBone {boneName=pelvis}}, {targetBone {boneName=dyn_thigh_01_l}}]
    constraintWeights [1,1]
   }
   n35 = "Constraint - dyn_skirt_fr2_02_l (PP)"!(ComponentPose=@n34, Alpha=0.8, ConstraintWeights_0=0, ConstraintWeights_1:float=1) -> (Pose) {
    alphaBoolBlend {blendOption=Linear}
    boneToModify {boneName=dyn_skirt_fr2_02_l}
    constraintSetup [{targetBone {boneName=pelvis}}, {targetBone {boneName=dyn_thigh_01_l}}]
    constraintWeights [1,1]
   }
   n36 = "Constraint - dyn_skirt_fr2_03_l (PP)"!(ComponentPose=@n35, Alpha=0.5, ConstraintWeights_0=0, ConstraintWeights_1:float=1) -> (Pose) {
    alphaBoolBlend {blendOption=Linear}
    boneToModify {boneName=dyn_skirt_fr2_03_l}
    constraintSetup [{targetBone {boneName=pelvis}}, {targetBone {boneName=dyn_thigh_01_l}}]
    constraintWeights [1,1]
   }
   n37 = "Constraint - dyn_skirt_fr_01_r_MSTR (PP)"!(ComponentPose=@n36, Alpha=0.9, ConstraintWeights_0=0, ConstraintWeights_1:float=1) -> (Pose) {
    alphaBoolBlend {blendOption=Linear}
    boneToModify {boneName=dyn_skirt_fr_01_r_MSTR}
    constraintSetup [{targetBone {boneName=pelvis}}, {targetBone {boneName=dyn_thigh_01_r}}]
    constraintWeights [1,1]
   }
   n38 = "Constraint - dyn_skirt_fr_01_r (PP)"!(ComponentPose=@n37, Alpha=0.9, ConstraintWeights_0=0, ConstraintWeights_1:float=1) -> (Pose) {
    alphaBoolBlend {blendOption=Linear}
    boneToModify {boneName=dyn_skirt_fr_01_r}
    constraintSetup [{targetBone {boneName=pelvis}}, {targetBone {boneName=dyn_thigh_01_r}}]
    constraintWeights [1,1]
   }
   n39 = "Constraint - dyn_skirt_fr_02_r (PP)"!(ComponentPose=@n38, Alpha=0.5, ConstraintWeights_0=0, ConstraintWeights_1:float=1) -> (Pose) {
    alphaBoolBlend {blendOption=Linear}
    boneToModify {boneName=dyn_skirt_fr_02_r}
    constraintSetup [{targetBone {boneName=pelvis}}, {targetBone {boneName=dyn_thigh_01_r}}]
    constraintWeights [1,1]
   }
   n40 = "Constraint - dyn_skirt_fr_03_r (PP)"!(ComponentPose=@n39, Alpha=0.5, ConstraintWeights_0=0, ConstraintWeights_1:float=1) -> (Pose) {
    alphaBoolBlend {blendOption=Linear}
    boneToModify {boneName=dyn_skirt_fr_03_r}
    constraintSetup [{targetBone {boneName=pelvis}}, {targetBone {boneName=dyn_thigh_01_r}}]
    constraintWeights [1,1]
   }
   n41 = "Constraint - dyn_skirt_fr2_01_r_MSTR (PP)"!(ComponentPose=@n40, Alpha=0.9, ConstraintWeights_0=0, ConstraintWeights_1:float=1) -> (Pose) {
    alphaBoolBlend {blendOption=Linear}
    boneToModify {boneName=dyn_skirt_fr2_01_r_MSTR}
    constraintSetup [{targetBone {boneName=pelvis}}, {targetBone {boneName=dyn_thigh_01_r}}]
    constraintWeights [1,1]
   }
   n42 = "Constraint - dyn_skirt_fr2_01_r (PP)"!(ComponentPose=@n41, Alpha=0.9, ConstraintWeights_0=0, ConstraintWeights_1:float=1) -> (Pose) {
    alphaBoolBlend {blendOption=Linear}
    boneToModify {boneName=dyn_skirt_fr2_01_r}
    constraintSetup [{targetBone {boneName=pelvis}}, {targetBone {boneName=dyn_thigh_01_r}}]
    constraintWeights [1,1]
   }
   n43 = "Constraint - dyn_skirt_fr2_02_r (PP)"!(ComponentPose=@n42, Alpha=0.5, ConstraintWeights_0=0, ConstraintWeights_1:float=1) -> (Pose) {
    alphaBoolBlend {blendOption=Linear}
    boneToModify {boneName=dyn_skirt_fr2_02_r}
    constraintSetup [{targetBone {boneName=pelvis}}, {targetBone {boneName=dyn_thigh_01_r}}]
    constraintWeights [1,1]
   }
   n44 = "Constraint - dyn_skirt_fr2_03_r (PP)"!(ComponentPose=@n43, Alpha=0.5, ConstraintWeights_0=0, ConstraintWeights_1:float=1) -> (Pose) {
    alphaBoolBlend {blendOption=Linear}
    boneToModify {boneName=dyn_skirt_fr2_03_r}
    constraintSetup [{targetBone {boneName=pelvis}}, {targetBone {boneName=dyn_thigh_01_r}}]
    constraintWeights [1,1]
   }
   n45 = "Constraint - dyn_skirt_fr_01_MSTR (PP)"!(ComponentPose=@n44, Alpha=1, ConstraintWeights_0=0.5, ConstraintWeights_1:float=0.5) -> (Pose) {
    alphaBoolBlend {blendOption=Linear}
    boneToModify {boneName=dyn_skirt_fr_01_MSTR}
    constraintSetup [{targetBone {boneName=dyn_skirt_fr_01_l_MSTR}}, {targetBone {boneName=dyn_skirt_fr_01_r_MSTR}}]
    constraintWeights [1,1]
   }
   n46 = "Constraint - dyn_skirt_fr_01 (PP)"!(ComponentPose=@n45, Alpha=1, ConstraintWeights_0=0.5, ConstraintWeights_1:float=0.5) -> (Pose) {
    alphaBoolBlend {blendOption=Linear}
    boneToModify {boneName=dyn_skirt_fr_01}
    constraintSetup [{targetBone {boneName=dyn_skirt_fr_01_l}}, {targetBone {boneName=dyn_skirt_fr_01_r}}]
    constraintWeights [1,1]
   }
   n47 = "Constraint - dyn_skirt_fr_02 (PP)"!(ComponentPose=@n46, Alpha=1, ConstraintWeights_0=0.5, ConstraintWeights_1:float=0.5) -> (Pose) {
    alphaBoolBlend {blendOption=Linear}
    boneToModify {boneName=dyn_skirt_fr_02}
    constraintSetup [{targetBone {boneName=dyn_skirt_fr_02_l}}, {targetBone {boneName=dyn_skirt_fr_02_r}}]
    constraintWeights [1,1]
   }
   n48 = "Constraint - dyn_skirt_fr_03 (PP)"!(ComponentPose=@n47, Alpha=1, ConstraintWeights_0=0.5, ConstraintWeights_1:float=0.5) -> (Pose) {
    alphaBoolBlend {blendOption=Linear}
    boneToModify {boneName=dyn_skirt_fr_03}
    constraintSetup [{targetBone {boneName=dyn_skirt_fr_03_l}}, {targetBone {boneName=dyn_skirt_fr_03_r}}]
    constraintWeights [1,1]
   }
   n49 = "Constraint - dyn_skirt_01_l_MSTR (P)"(ComponentPose=@n48, Alpha=0.5, ConstraintWeights_0=1) {
    alphaBoolBlend {blendOption=Linear}
    boneToModify {boneName=dyn_skirt_01_l_MSTR}
    constraintSetup [{targetBone {boneName=dyn_skirt_fr2_01_l_MSTR}}]
    constraintWeights [1]
   }
   n50 = "Constraint - dyn_skirt_01_l (P)"(ComponentPose=@n49, Alpha=0.5, ConstraintWeights_0=1) {
    alphaBoolBlend {blendOption=Linear}
    boneToModify {boneName=dyn_skirt_01_l}
    constraintSetup [{targetBone {boneName=dyn_skirt_fr2_01_l}}]
    constraintWeights [1]
   }
   n51 = "Constraint - dyn_skirt_02_l (P)"(ComponentPose=@n50, Alpha=0.5, ConstraintWeights_0=1) {
    alphaBoolBlend {blendOption=Linear}
    boneToModify {boneName=dyn_skirt_02_l}
    constraintSetup [{targetBone {boneName=dyn_skirt_fr2_02_l}}]
    constraintWeights [1]
   }
   n52 = "Constraint - dyn_skirt_03_l (P)"(ComponentPose=@n51, Alpha=0.5, ConstraintWeights_0=1) {
    alphaBoolBlend {blendOption=Linear}
    boneToModify {boneName=dyn_skirt_03_l}
    constraintSetup [{targetBone {boneName=dyn_skirt_fr2_03_l}}]
    constraintWeights [1]
   }
   n53 = "Constraint - dyn_skirt_bk2_01_l_MSTR (P)"(ComponentPose=@n52, Alpha=0.5, ConstraintWeights_0=1) {
    alphaBoolBlend {blendOption=Linear}
    boneToModify {boneName=dyn_skirt_bk2_01_l_MSTR}
    constraintSetup [{targetBone {boneName=dyn_skirt_01_l_MSTR}}]
    constraintWeights [1]
   }
   n54 = "Constraint - dyn_skirt_bk2_01_l (P)"(ComponentPose=@n53, Alpha=0.5, ConstraintWeights_0=1) {
    alphaBoolBlend {blendOption=Linear}
    boneToModify {boneName=dyn_skirt_bk2_01_l}
    constraintSetup [{targetBone {boneName=dyn_skirt_01_l}}]
    constraintWeights [1]
   }
   n55 = "Constraint - dyn_skirt_bk2_02_l (P)"(ComponentPose=@n54, Alpha=0.5, ConstraintWeights_0=1) {
    alphaBoolBlend {blendOption=Linear}
    boneToModify {boneName=dyn_skirt_bk2_02_l}
    constraintSetup [{targetBone {boneName=dyn_skirt_02_l}}]
    constraintWeights [1]
   }
   n56 = "Constraint - dyn_skirt_bk2_03_l (P)"(ComponentPose=@n55, Alpha=0.5, ConstraintWeights_0=1) {
    alphaBoolBlend {blendOption=Linear}
    boneToModify {boneName=dyn_skirt_bk2_03_l}
    constraintSetup [{targetBone {boneName=dyn_skirt_03_l}}]
    constraintWeights [1]
   }
   n57 = "Constraint - dyn_skirt_01_r_MSTR (P)"(ComponentPose=@n56, Alpha=0.5, ConstraintWeights_0=1) {
    alphaBoolBlend {blendOption=Linear}
    boneToModify {boneName=dyn_skirt_01_r_MSTR}
    constraintSetup [{targetBone {boneName=dyn_skirt_fr2_01_r_MSTR}}]
    constraintWeights [1]
   }
   n58 = "Constraint - dyn_skirt_01_r (P)"(ComponentPose=@n57, Alpha=0.5, ConstraintWeights_0=1) {
    alphaBoolBlend {blendOption=Linear}
    boneToModify {boneName=dyn_skirt_01_r}
    constraintSetup [{targetBone {boneName=dyn_skirt_fr2_01_r}}]
    constraintWeights [1]
   }
   n59 = "Constraint - dyn_skirt_02_r (P)"(ComponentPose=@n58, Alpha=0.5, ConstraintWeights_0=1) {
    alphaBoolBlend {blendOption=Linear}
    boneToModify {boneName=dyn_skirt_02_r}
    constraintSetup [{targetBone {boneName=dyn_skirt_fr2_02_r}}]
    constraintWeights [1]
   }
   n60 = "Constraint - dyn_skirt_03_r (P)"(ComponentPose=@n59, Alpha=0.5, ConstraintWeights_0=1) {
    alphaBoolBlend {blendOption=Linear}
    boneToModify {boneName=dyn_skirt_03_r}
    constraintSetup [{targetBone {boneName=dyn_skirt_fr2_03_r}}]
    constraintWeights [1]
   }
   n61 = "Constraint - dyn_skirt_bk2_01_r_MSTR (P)"(ComponentPose=@n60, Alpha=0.5, ConstraintWeights_0=1) {
    alphaBoolBlend {blendOption=Linear}
    boneToModify {boneName=dyn_skirt_bk2_01_r_MSTR}
    constraintSetup [{targetBone {boneName=dyn_skirt_01_r_MSTR}}]
    constraintWeights [1]
   }
   n62 = "Constraint - dyn_skirt_bk2_01_r (P)"(ComponentPose=@n61, Alpha=0.5, ConstraintWeights_0=1) {
    alphaBoolBlend {blendOption=Linear}
    boneToModify {boneName=dyn_skirt_bk2_01_r}
    constraintSetup [{targetBone {boneName=dyn_skirt_01_r}}]
    constraintWeights [1]
   }
   n63 = "Constraint - dyn_skirt_bk2_02_r (P)"(ComponentPose=@n62, Alpha=0.5, ConstraintWeights_0=1) {
    alphaBoolBlend {blendOption=Linear}
    boneToModify {boneName=dyn_skirt_bk2_02_r}
    constraintSetup [{targetBone {boneName=dyn_skirt_02_r}}]
    constraintWeights [1]
   }
   n64 = "Constraint - dyn_skirt_bk2_03_r (P)"(ComponentPose=@n63, Alpha=0.5, ConstraintWeights_0=1) {
    alphaBoolBlend {blendOption=Linear}
    boneToModify {boneName=dyn_skirt_bk2_03_r}
    constraintSetup [{targetBone {boneName=dyn_skirt_03_r}}]
    constraintWeights [1]
   }
   n65 = SelectFloat(A=1, B=1, bPickA=Get bIsCrouching())
   n66 = "Constraint - dyn_skirt_bk2_01_l_MSTR (PP)"!(ComponentPose=@n64, Alpha=@n65, ConstraintWeights_0=0.5, ConstraintWeights_1:float=0.5) -> (Pose) {
    alphaBoolBlend {blendOption=Linear}
    boneToModify {boneName=dyn_skirt_bk2_01_l_MSTR}
    constraintSetup [{targetBone {boneName=dyn_skirt_bk_01_l_MSTR}}, {targetBone {boneName=dyn_skirt_01_l_MSTR}}]
    constraintWeights [1,1]
   }
   n67 = "Constraint - dyn_skirt_bk2_01_l (PP)"!(ComponentPose=@n66, Alpha=@n65, ConstraintWeights_0=0.5, ConstraintWeights_1:float=0.5) -> (Pose) {
    alphaBoolBlend {blendOption=Linear}
    boneToModify {boneName=dyn_skirt_bk2_01_l}
    constraintSetup [{targetBone {boneName=dyn_skirt_bk_01_l}}, {targetBone {boneName=dyn_skirt_01_l}}]
    constraintWeights [1,1]
   }
   n68 = "Constraint - dyn_skirt_bk2_02_l (PP)"!(ComponentPose=@n67, Alpha=@n65, ConstraintWeights_0=0.5, ConstraintWeights_1:float=0.5) -> (Pose) {
    alphaBoolBlend {blendOption=Linear}
    boneToModify {boneName=dyn_skirt_bk2_02_l}
    constraintSetup [{targetBone {boneName=dyn_skirt_bk_02_l}}, {targetBone {boneName=dyn_skirt_02_l}}]
    constraintWeights [1,1]
   }
   n69 = "Constraint - dyn_skirt_bk2_03_l (PP)"!(ComponentPose=@n68, Alpha=@n65, ConstraintWeights_0=0.5, ConstraintWeights_1:float=0.5) -> (Pose) {
    alphaBoolBlend {blendOption=Linear}
    boneToModify {boneName=dyn_skirt_bk2_03_l}
    constraintSetup [{targetBone {boneName=dyn_skirt_bk_03_l}}, {targetBone {boneName=dyn_skirt_03_l}}]
    constraintWeights [1,1]
   }
   n70 = SelectFloat(A=1, B=1, bPickA=Get bIsCrouching())
   n71 = "Transform (Modify) Bone - Bone: dyn_skirt_bk2_01_l_MSTR"(ComponentPose=@n69, Alpha=@n70, Translation="2,0,0", Rotation="0,0,10", Scale="1,1,1") {
    rotationMode=BMM_Additive
    scaleMode=BMM_Additive
    translationMode=BMM_Additive
    alphaBoolBlend {blendOption=Linear}
    boneToModify {boneName=dyn_skirt_bk2_01_l_MSTR}
    rotation {roll=10}
    translation {x=2}
   }
   n72 = "Constraint - dyn_skirt_01_l_MSTR (PP)"!(ComponentPose=@n71, Alpha=@n70, ConstraintWeights_0=0.5, ConstraintWeights_1:float=0.5) -> (Pose) {
    alphaBoolBlend {blendOption=Linear}
    boneToModify {boneName=dyn_skirt_01_l_MSTR}
    constraintSetup [{targetBone {boneName=dyn_skirt_bk2_01_l_MSTR}}, {targetBone {boneName=dyn_skirt_fr2_01_l_MSTR}}]
    constraintWeights [1,1]
   }
   n73 = "Constraint - dyn_skirt_01_l (PP)"!(ComponentPose=@n72, Alpha=@n70, ConstraintWeights_0=0.5, ConstraintWeights_1:float=0.5) -> (Pose) {
    alphaBoolBlend {blendOption=Linear}
    boneToModify {boneName=dyn_skirt_01_l}
    constraintSetup [{targetBone {boneName=dyn_skirt_bk2_01_l}}, {targetBone {boneName=dyn_skirt_fr2_01_l_MSTR}}]
    constraintWeights [1,1]
   }
   n74 = "Constraint - dyn_skirt_02_l (PP)"!(ComponentPose=@n73, Alpha=@n70, ConstraintWeights_0=0.5, ConstraintWeights_1:float=0.5) -> (Pose) {
    alphaBoolBlend {blendOption=Linear}
    boneToModify {boneName=dyn_skirt_02_l}
    constraintSetup [{targetBone {boneName=dyn_skirt_bk2_02_l}}, {targetBone {boneName=dyn_skirt_fr2_02_l}}]
    constraintWeights [1,1]
   }
   n75 = "Constraint - dyn_skirt_03_l (PP)"!(ComponentPose=@n74, Alpha=@n70, ConstraintWeights_0=0.5, ConstraintWeights_1:float=0.5) -> (Pose) {
    alphaBoolBlend {blendOption=Linear}
    boneToModify {boneName=dyn_skirt_03_l}
    constraintSetup [{targetBone {boneName=dyn_skirt_bk2_03_l}}, {targetBone {boneName=dyn_skirt_fr2_03_l}}]
    constraintWeights [1,1]
   }
   n76 = SelectFloat(A=1, B=1, bPickA=Get bIsCrouching())
   n77 = "Transform (Modify) Bone - Bone: dyn_skirt_01_l_MSTR"(ComponentPose=@n75, Alpha=@n76, Translation="0,0,3", Rotation="0,0,0", Scale="1,1,1") {
    rotationMode=BMM_Additive
    rotationSpace=BCS_BoneSpace
    translationMode=BMM_Additive
    alphaBoolBlend {blendOption=Linear}
    boneToModify {boneName=dyn_skirt_01_l_MSTR}
    translation {z=3}
   }
   n78 = "Constraint - dyn_skirt_bk2_01_r_MSTR (PP)"!(ComponentPose=@n77, Alpha=@n76, ConstraintWeights_0=0.5, ConstraintWeights_1:float=0.5) -> (Pose) {
    alphaBoolBlend {blendOption=Linear}
    boneToModify {boneName=dyn_skirt_bk2_01_r_MSTR}
    constraintSetup [{targetBone {boneName=dyn_skirt_bk_01_r_MSTR}}, {targetBone {boneName=dyn_skirt_01_r_MSTR}}]
    constraintWeights [1,1]
   }
   n79 = "Constraint - dyn_skirt_bk2_01_r (PP)"!(ComponentPose=@n78, Alpha=@n76, ConstraintWeights_0=0.5, ConstraintWeights_1:float=0.5) -> (Pose) {
    alphaBoolBlend {blendOption=Linear}
    boneToModify {boneName=dyn_skirt_bk2_01_r}
    constraintSetup [{targetBone {boneName=dyn_skirt_bk_01_r}}, {targetBone {boneName=dyn_skirt_01_r}}]
    constraintWeights [1,1]
   }
   n80 = "Constraint - dyn_skirt_bk2_02_r (PP)"!(ComponentPose=@n79, Alpha=@n76, ConstraintWeights_0=0.5, ConstraintWeights_1:float=0.5) -> (Pose) {
    alphaBoolBlend {blendOption=Linear}
    boneToModify {boneName=dyn_skirt_bk2_02_r}
    constraintSetup [{targetBone {boneName=dyn_skirt_bk_02_r}}, {targetBone {boneName=dyn_skirt_02_r}}]
    constraintWeights [1,1]
   }
   n81 = "Constraint - dyn_skirt_bk2_03_r (PP)"!(ComponentPose=@n80, Alpha=@n76, ConstraintWeights_0=0.5, ConstraintWeights_1:float=0.5) -> (Pose) {
    alphaBoolBlend {blendOption=Linear}
    boneToModify {boneName=dyn_skirt_bk2_03_r}
    constraintSetup [{targetBone {boneName=dyn_skirt_bk_03_r}}, {targetBone {boneName=dyn_skirt_03_r}}]
    constraintWeights [1,1]
   }
   n82 = SelectFloat(A=1, B=1, bPickA=Get bIsCrouching())
   n83 = "Constraint - dyn_skirt_01_r_MSTR (PP)"!(ComponentPose=@n81, Alpha=@n82, ConstraintWeights_0=0.5, ConstraintWeights_1:float=0.5) -> (Pose) {
    alphaBoolBlend {blendOption=Linear}
    boneToModify {boneName=dyn_skirt_01_r_MSTR}
    constraintSetup [{targetBone {boneName=dyn_skirt_bk2_01_r_MSTR}}, {targetBone {boneName=dyn_skirt_fr2_01_r_MSTR}}]
    constraintWeights [1,1]
   }
   n84 = "Constraint - dyn_skirt_01_r (PP)"!(ComponentPose=@n83, Alpha=@n82, ConstraintWeights_0=0.5, ConstraintWeights_1:float=0.5) -> (Pose) {
    alphaBoolBlend {blendOption=Linear}
    boneToModify {boneName=dyn_skirt_01_r}
    constraintSetup [{targetBone {boneName=dyn_skirt_bk2_01_r}}, {targetBone {boneName=dyn_skirt_fr2_01_r}}]
    constraintWeights [1,1]
   }
   n85 = "Constraint - dyn_skirt_02_r (PP)"!(ComponentPose=@n84, Alpha=@n82, ConstraintWeights_0=0.5, ConstraintWeights_1:float=0.5) -> (Pose) {
    alphaBoolBlend {blendOption=Linear}
    boneToModify {boneName=dyn_skirt_02_r}
    constraintSetup [{targetBone {boneName=dyn_skirt_bk2_02_r}}, {targetBone {boneName=dyn_skirt_fr2_02_r}}]
    constraintWeights [1,1]
   }
   n86 = "Constraint - dyn_skirt_03_r (PP)"!(ComponentPose=@n85, Alpha=@n82, ConstraintWeights_0=0.5, ConstraintWeights_1:float=0.5) -> (Pose) {
    alphaBoolBlend {blendOption=Linear}
    boneToModify {boneName=dyn_skirt_03_r}
    constraintSetup [{targetBone {boneName=dyn_skirt_bk2_03_r}}, {targetBone {boneName=dyn_skirt_fr2_03_r}}]
    constraintWeights [1,1]
   }
   n87 = Anim Dynamics - Bone: dyn_skirt_fr_02_l(ComponentPose=@n86, Alpha=1, ExternalForce="0,0,0") {
    angularDampingOverride=0.9999
    angularSpringConstant=75
    bAngularSpring=true
    bGravityOverrideInSimSpace=true
    bOverrideAngularDamping=true
    simulationSpace=RootRelative
    alphaBoolBlend {blendOption=Linear}
    boundBone {boneName=dyn_skirt_fr_02_l}
    componentLinearAccScale=<1.2 1.2 1.2>
    componentLinearVelScale=<1.2 1.2 1.2>
    physicsBodyDefinitions [{
      sphereCollisionRadius=0
      boundBone {boneName=dyn_skirt_fr_02_l}
      boxExtents=<20 10 10>
      constraintSetup {angularLimitsMax=<25 0 25>, angularLimitsMin=<-25 -25 0>, angularTarget=<0 0 0>}
      localJointOffset=<2 0 0>
     }]
   }
   n88 = Anim Dynamics - Bone: dyn_skirt_fr_03_l(ComponentPose=@n87, Alpha=1, ExternalForce="0,0,0") {
    angularDampingOverride=0.9999
    angularSpringConstant=75
    bAngularSpring=true
    bGravityOverrideInSimSpace=true
    bOverrideAngularDamping=true
    simulationSpace=RootRelative
    alphaBoolBlend {blendOption=Linear}
    boundBone {boneName=dyn_skirt_fr_03_l}
    componentLinearAccScale=<1.2 1.2 1.2>
    componentLinearVelScale=<1.2 1.2 1.2>
    physicsBodyDefinitions [{
      sphereCollisionRadius=0
      boundBone {boneName=dyn_skirt_fr_03_l}
      boxExtents=<20 10 10>
      constraintSetup {angularLimitsMax=<25 0 25>, angularLimitsMin=<-25 -25 -25>, angularTarget=<0 0 0>}
      localJointOffset=<2 0 0>
     }]
   }
   n89 = Anim Dynamics - Bone: dyn_skirt_fr_02_r(ComponentPose=@n88, Alpha=1, ExternalForce="0,0,0") {
    angularDampingOverride=0.9999
    angularSpringConstant=75
    bAngularSpring=true
    bGravityOverrideInSimSpace=true
    bOverrideAngularDamping=true
    simulationSpace=RootRelative
    alphaBoolBlend {blendOption=Linear}
    boundBone {boneName=dyn_skirt_fr_02_r}
    componentLinearAccScale=<1.2 1.2 1.2>
    componentLinearVelScale=<1.2 1.2 1.2>
    physicsBodyDefinitions [{
      sphereCollisionRadius=0
      boundBone {boneName=dyn_skirt_fr_02_r}
      boxExtents=<20 20 10>
      constraintSetup {angularLimitsMax=<25 0 25>, angularLimitsMin=<-25 -25 -25>, angularTarget=<0 0 0>}
      localJointOffset=<2 0 0>
     }]
   }
   n90 = Anim Dynamics - Bone: dyn_skirt_fr_03_r(ComponentPose=@n89, Alpha=1, ExternalForce="0,0,0") {
    angularDampingOverride=0.9999
    angularSpringConstant=75
    bAngularSpring=true
    bGravityOverrideInSimSpace=true
    bOverrideAngularDamping=true
    simulationSpace=RootRelative
    alphaBoolBlend {blendOption=Linear}
    boundBone {boneName=dyn_skirt_fr_03_r}
    componentLinearAccScale=<1.2 1.2 1.2>
    componentLinearVelScale=<1.2 1.2 1.2>
    physicsBodyDefinitions [{
      sphereCollisionRadius=0
      boundBone {boneName=dyn_skirt_fr_03_r}
      boxExtents=<20 20 10>
      constraintSetup {angularLimitsMax=<25 0 25>, angularLimitsMin=<-25 -25 -25>, angularTarget=<0 0 0>}
      localJointOffset=<2 0 0>
     }]
   }
   n91 = Anim Dynamics - Bone: dyn_skirt_fr2_02_l(ComponentPose=@n90, Alpha=1, ExternalForce="0,0,0") {
    angularDampingOverride=0.9999
    angularSpringConstant=75
    bAngularSpring=true
    bGravityOverrideInSimSpace=true
    bOverrideAngularDamping=true
    simulationSpace=RootRelative
    alphaBoolBlend {blendOption=Linear}
    boundBone {boneName=dyn_skirt_fr2_02_l}
    componentLinearAccScale=<1.2 1.2 1.2>
    componentLinearVelScale=<1.2 1.2 1.2>
    physicsBodyDefinitions [{
      sphereCollisionRadius=0
      boundBone {boneName=dyn_skirt_fr2_02_l}
      boxExtents=<20 10 10>
      constraintSetup {angularLimitsMax=<25 0 25>, angularLimitsMin=<-25 -25 -25>, angularTarget=<0 0 0>}
      localJointOffset=<2 0 0>
     }]
   }
   n92 = Anim Dynamics - Bone: dyn_skirt_fr2_03_l(ComponentPose=@n91, Alpha=1, ExternalForce="0,0,0") {
    angularDampingOverride=0.9999
    angularSpringConstant=75
    bAngularSpring=true
    bGravityOverrideInSimSpace=true
    bOverrideAngularDamping=true
    simulationSpace=RootRelative
    alphaBoolBlend {blendOption=Linear}
    boundBone {boneName=dyn_skirt_fr2_03_l}
    componentLinearAccScale=<1.2 1.2 1.2>
    componentLinearVelScale=<1.2 1.2 1.2>
    physicsBodyDefinitions [{
      sphereCollisionRadius=0
      boundBone {boneName=dyn_skirt_fr2_03_l}
      boxExtents=<20 10 10>
      constraintSetup {angularLimitsMax=<25 0 25>, angularLimitsMin=<-25 -25 -25>, angularTarget=<0 0 0>}
      localJointOffset=<2 0 0>
     }]
   }
   n93 = Anim Dynamics - Bone: dyn_skirt_fr2_02_r(ComponentPose=@n92, Alpha=1, ExternalForce="0,0,0") {
    angularDampingOverride=0.9999
    angularSpringConstant=75
    bAngularSpring=true
    bGravityOverrideInSimSpace=true
    bOverrideAngularDamping=true
    simulationSpace=RootRelative
    alphaBoolBlend {blendOption=Linear}
    boundBone {boneName=dyn_skirt_fr2_02_r}
    componentLinearAccScale=<1.2 1.2 1.2>
    componentLinearVelScale=<1.2 1.2 1.2>
    physicsBodyDefinitions [{
      sphereCollisionRadius=0
      boundBone {boneName=dyn_skirt_fr2_02_r}
      boxExtents=<20 20 10>
      constraintSetup {angularLimitsMax=<25 0 25>, angularLimitsMin=<-25 -25 -25>, angularTarget=<0 0 0>}
      localJointOffset=<2 0 0>
     }]
   }
   n94 = Anim Dynamics - Bone: dyn_skirt_fr2_03_r(ComponentPose=@n93, Alpha=1, ExternalForce="0,0,0") {
    angularDampingOverride=0.9999
    angularSpringConstant=75
    bAngularSpring=true
    bGravityOverrideInSimSpace=true
    bOverrideAngularDamping=true
    simulationSpace=RootRelative
    alphaBoolBlend {blendOption=Linear}
    boundBone {boneName=dyn_skirt_fr2_03_r}
    componentLinearAccScale=<1.2 1.2 1.2>
    componentLinearVelScale=<1.2 1.2 1.2>
    physicsBodyDefinitions [{
      sphereCollisionRadius=0
      boundBone {boneName=dyn_skirt_fr2_03_r}
      boxExtents=<20 20 10>
      constraintSetup {angularLimitsMax=<25 0 25>, angularLimitsMin=<-25 -25 -25>, angularTarget=<0 0 0>}
      localJointOffset=<2 0 0>
     }]
   }
   n95 = Anim Dynamics - Bone: dyn_skirt_02_l(ComponentPose=@n94, Alpha=1, ExternalForce="0,0,0") {
    angularDampingOverride=0.9999
    angularSpringConstant=75
    bAngularSpring=true
    bGravityOverrideInSimSpace=true
    bOverrideAngularDamping=true
    simulationSpace=RootRelative
    alphaBoolBlend {blendOption=Linear}
    boundBone {boneName=dyn_skirt_02_l}
    componentLinearAccScale=<1.2 1.2 1.2>
    componentLinearVelScale=<1.2 1.2 1.2>
    physicsBodyDefinitions [{
      sphereCollisionRadius=0
      boundBone {boneName=dyn_skirt_02_l}
      boxExtents=<20 10 10>
      constraintSetup {angularLimitsMax=<25 0 25>, angularLimitsMin=<-25 -25 -25>, angularTarget=<0 0 0>}
      localJointOffset=<2 0 0>
     }]
   }
   n96 = Anim Dynamics - Bone: dyn_skirt_03_l(ComponentPose=@n95, Alpha=1, ExternalForce="0,0,0") {
    angularDampingOverride=0.9999
    angularSpringConstant=75
    bAngularSpring=true
    bGravityOverrideInSimSpace=true
    bOverrideAngularDamping=true
    simulationSpace=RootRelative
    alphaBoolBlend {blendOption=Linear}
    boundBone {boneName=dyn_skirt_03_l}
    componentLinearAccScale=<1.2 1.2 1.2>
    componentLinearVelScale=<1.2 1.2 1.2>
    physicsBodyDefinitions [{
      sphereCollisionRadius=0
      boundBone {boneName=dyn_skirt_03_l}
      boxExtents=<20 10 10>
      constraintSetup {angularLimitsMax=<25 0 25>, angularLimitsMin=<-25 -25 -25>, angularTarget=<0 0 0>}
      localJointOffset=<2 0 0>
     }]
   }
   n97 = Anim Dynamics - Bone: dyn_skirt_01_r(ComponentPose=@n96, Alpha=1, ExternalForce="0,0,0") {
    angularDampingOverride=0.9999
    angularSpringConstant=75
    bAngularSpring=true
    bGravityOverrideInSimSpace=true
    bOverrideAngularDamping=true
    simulationSpace=RootRelative
    alphaBoolBlend {blendOption=Linear}
    boundBone {boneName=dyn_skirt_01_r}
    componentLinearAccScale=<1.2 1.2 1.2>
    componentLinearVelScale=<1.2 1.2 1.2>
    physicsBodyDefinitions [{
      sphereCollisionRadius=0
      boundBone {boneName=dyn_skirt_01_r}
      boxExtents=<20 20 10>
      constraintSetup {angularLimitsMax=<25 0 25>, angularLimitsMin=<-25 -25 -25>, angularTarget=<0 0 0>}
      localJointOffset=<2 0 0>
     }]
   }
   n98 = Anim Dynamics - Bone: dyn_skirt_03_r(ComponentPose=@n97, Alpha=1, ExternalForce="0,0,0") {
    angularDampingOverride=0.9999
    angularSpringConstant=75
    bAngularSpring=true
    bGravityOverrideInSimSpace=true
    bOverrideAngularDamping=true
    simulationSpace=RootRelative
    alphaBoolBlend {blendOption=Linear}
    boundBone {boneName=dyn_skirt_03_r}
    componentLinearAccScale=<1.2 1.2 1.2>
    componentLinearVelScale=<1.2 1.2 1.2>
    physicsBodyDefinitions [{
      sphereCollisionRadius=0
      boundBone {boneName=dyn_skirt_03_r}
      boxExtents=<20 20 10>
      constraintSetup {angularLimitsMax=<25 0 25>, angularLimitsMin=<-25 -25 -25>, angularTarget=<0 0 0>}
      localJointOffset=<2 0 0>
     }]
   }
   n99 = Anim Dynamics - Bone: dyn_skirt_bk2_02_r(ComponentPose=@n98, Alpha=1, ExternalForce="0,0,0") {
    angularDampingOverride=0.9999
    angularSpringConstant=75
    bAngularSpring=true
    bGravityOverrideInSimSpace=true
    bOverrideAngularDamping=true
    simulationSpace=RootRelative
    alphaBoolBlend {blendOption=Linear}
    boundBone {boneName=dyn_skirt_bk2_02_r}
    componentLinearAccScale=<1.2 1.2 1.2>
    componentLinearVelScale=<1.2 1.2 1.2>
    physicsBodyDefinitions [{
      sphereCollisionRadius=0
      boundBone {boneName=dyn_skirt_bk2_02_r}
      boxExtents=<20 20 10>
      constraintSetup {angularLimitsMax=<25 0 25>, angularLimitsMin=<-25 -25 -25>, angularTarget=<0 0 0>}
      localJointOffset=<2 0 0>
     }]
   }
   n100 = Anim Dynamics - Bone: dyn_skirt_bk2_02_l(ComponentPose=@n99, Alpha=1, ExternalForce="0,0,0") {
    angularDampingOverride=0.9999
    angularSpringConstant=75
    bAngularSpring=true
    bGravityOverrideInSimSpace=true
    bOverrideAngularDamping=true
    simulationSpace=RootRelative
    alphaBoolBlend {blendOption=Linear}
    boundBone {boneName=dyn_skirt_bk2_02_l}
    componentLinearAccScale=<1.2 1.2 1.2>
    componentLinearVelScale=<1.2 1.2 1.2>
    physicsBodyDefinitions [{
      sphereCollisionRadius=0
      boundBone {boneName=dyn_skirt_bk2_02_l}
      boxExtents=<20 20 10>
      constraintSetup {angularLimitsMax=<25 0 25>, angularLimitsMin=<-25 -25 -25>, angularTarget=<0 0 0>}
      localJointOffset=<2 0 0>
     }]
   }
   n101 = Anim Dynamics - Bone: dyn_skirt_bk2_03_r(ComponentPose=@n100, Alpha=1, ExternalForce="0,0,0") {
    angularDampingOverride=0.9999
    angularSpringConstant=75
    bAngularSpring=true
    bGravityOverrideInSimSpace=true
    bOverrideAngularDamping=true
    simulationSpace=RootRelative
    alphaBoolBlend {blendOption=Linear}
    boundBone {boneName=dyn_skirt_bk2_03_r}
    componentLinearAccScale=<1.2 1.2 1.2>
    componentLinearVelScale=<1.2 1.2 1.2>
    physicsBodyDefinitions [{
      sphereCollisionRadius=0
      boundBone {boneName=dyn_skirt_bk2_03_r}
      boxExtents=<20 20 10>
      constraintSetup {angularLimitsMax=<25 0 25>, angularLimitsMin=<-25 -25 -25>, angularTarget=<0 0 0>}
      localJointOffset=<2 0 0>
     }]
   }
   n102 = Anim Dynamics - Bone: dyn_skirt_bk2_03_l(ComponentPose=@n101, Alpha=1, ExternalForce="0,0,0") {
    angularDampingOverride=0.9999
    angularSpringConstant=75
    bAngularSpring=true
    bGravityOverrideInSimSpace=true
    bOverrideAngularDamping=true
    simulationSpace=RootRelative
    alphaBoolBlend {blendOption=Linear}
    boundBone {boneName=dyn_skirt_bk2_03_l}
    componentLinearAccScale=<1.2 1.2 1.2>
    componentLinearVelScale=<1.2 1.2 1.2>
    physicsBodyDefinitions [{
      sphereCollisionRadius=0
      boundBone {boneName=dyn_skirt_bk2_03_l}
      boxExtents=<20 20 10>
      constraintSetup {angularLimitsMax=<25 0 25>, angularLimitsMin=<-25 -25 -25>, angularTarget=<0 0 0>}
      localJointOffset=<2 0 0>
     }]
   }
   n103 = "Transform (Modify) Bone - Bone: dyn_skirt_01_r_MSTR"(ComponentPose=@n102, Alpha=SelectFloat(A=1, B=1, bPickA=Get bIsCrouching()), Translation="0,0,2", Rotation="0,0,0", Scale="1,1,1") {
    rotationMode=BMM_Additive
    rotationSpace=BCS_BoneSpace
    scaleMode=BMM_Additive
    scaleSpace=BCS_BoneSpace
    translationMode=BMM_Additive
    translationSpace=BCS_BoneSpace
    alphaBoolBlend {blendOption=Linear}
    boneToModify {boneName=dyn_skirt_01_r_MSTR}
    translation {z=2}
   }
   Output Pose(Result=Component To Local(ComponentPose=@n103)) {layerGroup=None}
   n104 = "Constraint - dyn_pin_mstr (PPP)"!(
    ComponentPose=Local To Component(LocalPose=Copy Pose From Mesh() {bUseAttachedParent=true})
    Alpha=1
    ConstraintWeights_0=0.7551
    ConstraintWeights_1:float=0.1965
    ConstraintWeights_2:float=0.0484
   ) -> (Pose) {
    alphaBoolBlend {blendOption=Linear}
    boneToModify {boneName=dyn_pin_mstr}
    constraintSetup [{targetBone {boneName=spine_04}}, {targetBone {boneName=spine_05}}, {targetBone {boneName=clavicle_l}}]
    constraintWeights [1,1,1]
   }
   n105 = "Constraint - dyn_thigh_01_l (RR)"!(ComponentPose=@n104, Alpha=1, ConstraintWeights_0=1, ConstraintWeights_1:float=0) -> (Pose) {
    alphaBoolBlend {blendOption=Linear}
    boneToModify {boneName=dyn_thigh_01_l}
    constraintSetup [{transformType=Rotation, targetBone {boneName=thigh_l}}, {transformType=Rotation, targetBone {boneName=pelvis}}]
    constraintWeights [1,1]
   }
   n106 = "Constraint - dyn_thigh_01_r (RR)"!(ComponentPose=@n105, Alpha=1, ConstraintWeights_0=1, ConstraintWeights_1:float=0) -> (Pose) {
    alphaBoolBlend {blendOption=Linear}
    boneToModify {boneName=dyn_thigh_01_r}
    constraintSetup [{transformType=Rotation, targetBone {boneName=thigh_r}}, {transformType=Rotation, targetBone {boneName=pelvis}}]
    constraintWeights [1,1]
   }
   n107 = "Constraint - dyn_skirt_fr_01_MSTR (RR)"!(ComponentPose=@n106, Alpha=1, ConstraintWeights_0=0.5, ConstraintWeights_1:float=0.5) -> (Pose) {
    alphaBoolBlend {blendOption=Linear}
    boneToModify {boneName=dyn_skirt_fr_01_MSTR}
    constraintSetup [
     {transformType=Rotation, targetBone {boneName=dyn_skirt_fr_01_l_MSTR}}
     {transformType=Rotation, targetBone {boneName=dyn_skirt_fr_01_r_MSTR}}
    ]
    constraintWeights [1,1]
   }
   n108 = "Constraint - dyn_skirt_fr_01 (RR)"!(ComponentPose=@n107, Alpha=1, ConstraintWeights_0=0.5, ConstraintWeights_1:float=0.5) -> (Pose) {
    alphaBoolBlend {blendOption=Linear}
    boneToModify {boneName=dyn_skirt_fr_01}
    constraintSetup [{transformType=Rotation, targetBone {boneName=dyn_skirt_fr_01_l}}, {transformType=Rotation, targetBone {boneName=dyn_skirt_fr_01_r}}]
    constraintWeights [1,1]
   }
   n109 = "Constraint - dyn_skirt_fr_02 (RR)"!(ComponentPose=@n108, Alpha=1, ConstraintWeights_0=0.5, ConstraintWeights_1:float=0.5) -> (Pose) {
    alphaBoolBlend {blendOption=Linear}
    boneToModify {boneName=dyn_skirt_fr_02}
    constraintSetup [{transformType=Rotation, targetBone {boneName=dyn_skirt_fr_02_l}}, {transformType=Rotation, targetBone {boneName=dyn_skirt_fr_02_r}}]
    constraintWeights [1,1]
   }
   n110 = "Constraint - dyn_skirt_fr_03 (RR)"!(ComponentPose=@n109, Alpha=1, ConstraintWeights_0=0.5, ConstraintWeights_1:float=0.5) -> (Pose) {
    alphaBoolBlend {blendOption=Linear}
    boneToModify {boneName=dyn_skirt_fr_03}
    constraintSetup [{transformType=Rotation, targetBone {boneName=dyn_skirt_fr_03_l}}, {transformType=Rotation, targetBone {boneName=dyn_skirt_fr_03_r}}]
    constraintWeights [1,1]
   }
   Save cached pose 'PreDressRBAN'(Pose=Component To Local(ComponentPose=@n110)) {cacheName=PreDressRBAN}
  ]}
 EventGraph {nodes [
   Event BlueprintUpdateAnimation()
   > Sequence()
   > n1 = Gravity Override State Function DEPRECATED!(
    execute
    CharacterPartAnimInstance=Self-Reference()
    GameTimeInSeconds=GetGameTimeInSeconds()
    GravityOverrideParams_Skydive_GoingUp_28_#1_WindFrequency_28_#2:Vector="1.5,7,8"
    GravityOverrideParams_Skydive_GoingUp_28_#1_WindAmplitude_31_#3:Vector="200,-500,400"
    GravityOverrideParams_Skydive_GoingUp_28_#1_WindOffset_32_#4:Vector="0,0,0"
    GravityOverrideParams_Skydive_GoingUp_28_#1_GravityOverride_15_#5:Vector="0,0,-100"
    GravityOverrideParams_Skydive_GoingUp_28_#1_JointName_27_#6:Name=root
    GravityOverrideParams_Skydive_FloatingDown_29_#7_WindFrequency_28_#2:Vector="1.5,7,8"
    GravityOverrideParams_Skydive_FloatingDown_29_#7_WindAmplitude_31_#3:Vector="200,-500,400"
    GravityOverrideParams_Skydive_FloatingDown_29_#7_WindOffset_32_#4:Vector="0,-600,400"
    GravityOverrideParams_Skydive_FloatingDown_29_#7_GravityOverride_15_#5:Vector="0,-700,600"
    GravityOverrideParams_Skydive_FloatingDown_29_#7_JointName_27_#6:Name=root
    GravityOverrideParams_Skydive_DivingDown_30_#8_WindFrequency_28_#2:Vector="1.5,7,8"
    GravityOverrideParams_Skydive_DivingDown_30_#8_WindAmplitude_31_#3:Vector="50,300,100"
    GravityOverrideParams_Skydive_DivingDown_30_#8_WindOffset_32_#4:Vector="0,0,0"
    GravityOverrideParams_Skydive_DivingDown_30_#8_GravityOverride_15_#5:Vector="0,0,0"
    GravityOverrideParams_Skydive_DivingDown_30_#8_JointName_27_#6:Name=pelvis
    GravityOverrideParams_Skydive_Parachuting_31_#9_WindFrequency_28_#2:Vector="1.5,7,8"
    GravityOverrideParams_Skydive_Parachuting_31_#9_WindAmplitude_31_#3:Vector="200,-500,400"
    GravityOverrideParams_Skydive_Parachuting_31_#9_WindOffset_32_#4:Vector="0,0,0"
    GravityOverrideParams_Skydive_Parachuting_31_#9_GravityOverride_15_#5:Vector="0,-800,-980"
    GravityOverrideParams_Skydive_Parachuting_31_#9_JointName_27_#6:Name=root
    GravityOverrideParams_OnGround_Standing_32_#10:Gravity Override Setting Struct Structure=(WindFrequency_28_#2=(X=0,Y=0,Z=0),WindAmplitude_31_#3=(X=0,Y=0,Z=0),WindOffset_32_#4=(X=0,Y=0,Z=0),GravityOverride_15_#5=(X=0,Y=0,Z=-980),JointName_27_#6="root")
    GravityOverrideParams_OnGround_Moving_33_#11:Gravity Override Setting Struct Structure=(WindFrequency_28_#2=(X=0,Y=0,Z=0),WindAmplitude_31_#3=(X=0,Y=0,Z=0),WindOffset_32_#4=(X=0,Y=0,Z=0),GravityOverride_15_#5=(X=0,Y=0,Z=-980),JointName_27_#6="root")
    GravityOverrideParams_EmoteOrMelee_34_#12:Gravity Override Setting Struct Structure=(WindFrequency_28_#2=(X=0,Y=0,Z=0),WindAmplitude_31_#3=(X=0,Y=0,Z=0),WindOffset_32_#4=(X=0,Y=0,Z=0),GravityOverride_15_#5=(X=0,Y=0,Z=-980),JointName_27_#6="root")
    GravityOverrideParams_Swimming_35_#13:Gravity Override Setting Struct Structure=(WindFrequency_28_#2=(X=0,Y=0,Z=0),WindAmplitude_31_#3=(X=0,Y=0,Z=0),WindOffset_32_#4=(X=0,Y=0,Z=0),GravityOverride_15_#5=(X=-800,Y=0,Z=0),JointName_27_#6="pelvis")
    GravityOverrideParams_Front_End_36_#14:Gravity Override Setting Struct Structure=(WindFrequency_28_#2=(X=0,Y=0,Z=0),WindAmplitude_31_#3=(X=0,Y=0,Z=0),WindOffset_32_#4=(X=0,Y=0,Z=0),GravityOverride_15_#5=(X=0,Y=0,Z=-980),JointName_27_#6="root")
    GravityOverrideParams_DBNO_40_#15:Gravity Override Setting Struct Structure=(WindFrequency_28_#2=(X=0,Y=0,Z=0),WindAmplitude_31_#3=(X=0,Y=0,Z=0),WindOffset_32_#4=(X=0,Y=0,Z=0),GravityOverride_15_#5=(X=-980,Y=400,Z=0),JointName_27_#6="pelvis")
    GravityOverrideParams_DBNO_Carried_41_#16:Gravity Override Setting Struct Structure=(WindFrequency_28_#2=(X=0,Y=0,Z=0),WindAmplitude_31_#3=(X=0,Y=0,Z=0),WindOffset_32_#4=(X=0,Y=0,Z=0),GravityOverride_15_#5=(X=-980,Y=500,Z=-500),JointName_27_#6="pelvis")
    GravityOverrideParams_OnGround_Falling_45_#17:Gravity Override Setting Struct Structure=(WindFrequency_28_#2=(X=0,Y=0,Z=0),WindAmplitude_31_#3=(X=0,Y=0,Z=0),WindOffset_32_#4=(X=0,Y=0,Z=0),GravityOverride_15_#5=(X=0,Y=0,Z=-980),JointName_27_#6="root")
    GravityOverrideParams_Ziplining_46_#18:Gravity Override Setting Struct Structure=(WindFrequency_28_#2=(X=0,Y=0,Z=0),WindAmplitude_31_#3=(X=0,Y=0,Z=0),WindOffset_32_#4=(X=0,Y=0,Z=0),GravityOverride_15_#5=(X=0,Y=0,Z=-980),JointName_27_#6="root")
    GravityOverrideParams_Skydive_Surfing_49_#19:Gravity Override Setting Struct Structure=(WindFrequency_28_#2=(X=1.5,Y=7,Z=8),WindAmplitude_31_#3=(X=200,Y=-500,Z=400),WindOffset_32_#4=(X=0,Y=0,Z=0),GravityOverride_15_#5=(X=500,Y=-300,Z=-980),JointName_27_#6="root")
    GravityOverrideParams_Riding_Standing_54_#20:Gravity Override Setting Struct Structure=(WindFrequency_28_#2=(X=0,Y=0,Z=0),WindAmplitude_31_#3=(X=0,Y=0,Z=0),WindOffset_32_#4=(X=0,Y=0,Z=0),GravityOverride_15_#5=(X=0,Y=-350,Z=-980),JointName_27_#6="root")
    GravityOverrideParams_Riding_Moving_55_#21:Gravity Override Setting Struct Structure=(WindFrequency_28_#2=(X=0,Y=0,Z=0),WindAmplitude_31_#3=(X=0,Y=0,Z=0),WindOffset_32_#4=(X=0,Y=0,Z=0),GravityOverride_15_#5=(X=0,Y=-450,Z=-450),JointName_27_#6="root")
    GravityOverrideParams_Riding_Moving_Targeting_58_#22:Gravity Override Setting Struct Structure=(WindFrequency_28_#2=(X=0,Y=0,Z=0),WindAmplitude_31_#3=(X=0,Y=0,Z=0),WindOffset_32_#4=(X=0,Y=0,Z=0),GravityOverride_15_#5=(X=0,Y=-200,Z=-980),JointName_27_#6="root")
    GravityOverrideParams_Tactical_Sprint_57_#23:Gravity Override Setting Struct Structure=(WindFrequency_28_#2=(X=0,Y=0,Z=0),WindAmplitude_31_#3=(X=0,Y=0,Z=0),WindOffset_32_#4=(X=0,Y=0,Z=0),GravityOverride_15_#5=(X=0,Y=-450,Z=-500),JointName_27_#6="root")
    GravityOverrideParams_Rail_Grinding_61_#24:Gravity Override Setting Struct Structure=(WindFrequency_28_#2=(X=0,Y=0,Z=0),WindAmplitude_31_#3=(X=0,Y=0,Z=0),WindOffset_32_#4=(X=0,Y=0,Z=0),GravityOverride_15_#5=(X=0,Y=0,Z=-980),JointName_27_#6="root")
    GravityOverrideParams_DBNO_Tumbling_67_#25:Gravity Override Setting Struct Structure=(WindFrequency_28_#2=(X=0,Y=0,Z=0),WindAmplitude_31_#3=(X=0,Y=0,Z=0),WindOffset_32_#4=(X=0,Y=0,Z=0),GravityOverride_15_#5=(X=-2000,Y=0,Z=0),JointName_27_#6="pelvis")
    GravityOverrideParams_DBNO_RollFwd_68_#26:Gravity Override Setting Struct Structure=(WindFrequency_28_#2=(X=0,Y=0,Z=0),WindAmplitude_31_#3=(X=0,Y=0,Z=0),WindOffset_32_#4=(X=0,Y=0,Z=0),GravityOverride_15_#5=(X=-2000,Y=0,Z=0),JointName_27_#6="pelvis")
    GravityOverrideParams_DBNO_RollBwd_69_#27:Gravity Override Setting Struct Structure=(WindFrequency_28_#2=(X=0,Y=0,Z=0),WindAmplitude_31_#3=(X=0,Y=0,Z=0),WindOffset_32_#4=(X=0,Y=0,Z=0),GravityOverride_15_#5=(X=-4000,Y=0,Z=0),JointName_27_#6="pelvis")
    GravityOverrideParams_DBNO_Impulse_70_#28:Gravity Override Setting Struct Structure=(WindFrequency_28_#2=(X=0,Y=0,Z=0),WindAmplitude_31_#3=(X=0,Y=0,Z=0),WindOffset_32_#4=(X=0,Y=0,Z=0),GravityOverride_15_#5=(X=-2000,Y=0,Z=0),JointName_27_#6="pelvis")
    RBAN Params
    DebugDraw
    DebugJointName
   ) -> (then, GravityOverride, CurrentStateOut, GravityOverrideSettingsOut, RBAN SettingsOut)
   > Set Gravity Override(Gravity Override=@n1)
  ]}
]
variables [Gravity Override {type=Vector}]
