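# Physics asset for the front hair part of the female medium Musician character, set up for secondary hair motion.
# Left and right hair strands are each four capsules chained from a small master sphere by constraints with swing and tight twist limits, tuned differently per side.
# Unconstrained capsules and a sphere on the upper spine, neck, clavicles and upper arms give the hair body shapes to collide with.
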PhysicsAsset /BRCosmetics/Characters/Player/Female/Medium/Bodies/F_MED_Musician/Meshes/Parts/F_MED_Musician_FaceAcc_FrontHair_Physics
bodies {
 clavicle_l [{shapeType=Capsule, length=6, radius=4.5, shapeName=clavicle_l_capsule, center=<5.6674 0 0>, rotation=<rot -90 0.1189 -0.1194>}]
 clavicle_r [{shapeType=Capsule, length=12.5, radius=4.5, shapeName=clavicle_r_capsule, center=<-10.1079 0.0002 0>, rotation=<rot -90 0.0009 -0.0021>}]
 dyn_hair_01_l [{shapeType=Capsule, length=1.5, radius=2, shapeName=dyn_hair_01_l_capsule, center=<4.0092 0.0001 0>, rotation=<rot 90 180 179.9992>}]
 dyn_hair_01_r [{shapeType=Capsule, length=3, radius=1.5, shapeName=dyn_hair_01_r_capsule, center=<3.1002 -0.0001 0>, rotation=<rot 90 -166.7374 -166.7365>}]
 dyn_hair_02_l [{shapeType=Capsule, length=2.5, radius=3, shapeName=dyn_hair_02_l_capsule, center=<3.5765 0 0>, rotation=<rot 90 -163.0017 -163.0019>}]
 dyn_hair_02_r [{shapeType=Capsule, length=3.5, radius=1.5, shapeName=dyn_hair_02_r_capsule, center=<2.6569 0 0>, rotation=<rot 90 -0.0002 0.0003>}]
 dyn_hair_03_l [{shapeType=Capsule, length=2.5, radius=2, shapeName=dyn_hair_03_l_capsule, center=<4.4256 0 0>, rotation=<rot -0.0015 -179.9982 -120.0005>}]
 dyn_hair_03_r [{shapeType=Capsule, length=2, radius=2.5, shapeName=dyn_hair_03_r_capsule, center=<3.0009 0 0>, rotation=<rot 90 180 -179.9997>}]
 dyn_hair_04_l [{shapeType=Capsule, length=4, radius=1.5, shapeName=dyn_hair_04_l_capsule, center=<3.9659 0 0>, rotation=<rot -90 179.9473 -179.947>}]
 dyn_hair_04_r [{shapeType=Capsule, length=3, radius=2.5, shapeName=dyn_hair_04_r_capsule, center=<3.4413 0 0>, rotation=<rot 90 144.7357 144.736>}]
 dyn_hair_MSTR_l [{shapeType=Sphere, radius=1, shapeName=dyn_hair_MSTR_l_sphere}]
 dyn_hair_MSTR_r [{shapeType=Sphere, radius=1, shapeName=dyn_hair_MSTR_r_sphere}]
 neck_01 [{shapeType=Capsule, length=9, radius=5, shapeName=neck_01_capsule, center=<0.1181 1.5272 0>, rotation=<rot -90 90 -110.0007>}]
 neck_02 [{shapeType=Sphere, radius=5.5, shapeName=neck_02_sphere, center=<-6.7011 -7.5408 -10.3203>}]
 spine_04 [{shapeType=Capsule, length=16, radius=8, shapeName=spine_04_capsule, center=<4.4241 1.289 0>}]
 spine_05 [{shapeType=Capsule, length=24.5, radius=5.5, shapeName=spine_05_capsule, center=<5.5302 -1.6357 0>}]
 upperarm_l [{shapeType=Capsule, length=25.5, radius=4.5, shapeName=upperarm_l_capsule, center=<15.635 0.0002 0>, rotation=<rot 90 0.0004 -0.0002>}]
 upperarm_r [{shapeType=Capsule, length=25.5, radius=4.5, shapeName=upperarm_r_capsule, center=<-16.5125 0 0>, rotation=<rot -90 180 180>}]
}
constraints [
 {bone1Name=dyn_hair_01_l, bone2Name=dyn_hair_MSTR_l, swing1LimitDegrees=35, swing2LimitDegrees=12, twistLimitDegrees=3}
 {bone1Name=dyn_hair_02_l, bone2Name=dyn_hair_01_l, swing1LimitDegrees=15, swing2LimitDegrees=10, twistLimitDegrees=5}
 {bone1Name=dyn_hair_03_l, bone2Name=dyn_hair_02_l, swing1LimitDegrees=20, swing2LimitDegrees=20, twistLimitDegrees=5}
 {bone1Name=dyn_hair_04_l, bone2Name=dyn_hair_03_l, swing1LimitDegrees=20, swing2LimitDegrees=20, twistLimitDegrees=6}
 {bone1Name=dyn_hair_01_r, bone2Name=dyn_hair_MSTR_r, swing1LimitDegrees=20, swing2LimitDegrees=10, twistLimitDegrees=5}
 {bone1Name=dyn_hair_02_r, bone2Name=dyn_hair_01_r, swing1LimitDegrees=20, swing2LimitDegrees=20, twistLimitDegrees=5}
 {bone1Name=dyn_hair_03_r, bone2Name=dyn_hair_02_r, swing1LimitDegrees=20, swing2LimitDegrees=20, twistLimitDegrees=5}
 {bone1Name=dyn_hair_04_r, bone2Name=dyn_hair_03_r, swing1LimitDegrees=20, swing2LimitDegrees=20, twistLimitDegrees=5}
]
