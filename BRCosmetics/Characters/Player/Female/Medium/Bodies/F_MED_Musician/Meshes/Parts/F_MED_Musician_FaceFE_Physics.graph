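# Physics asset for the female medium Musician character's FaceFE mesh, set up for dynamic hair.
# Center, left and right hair chains hang from small master bodies linked to the head, plus a short "br" branch off the center master.
# Hair segments have limited swing and twist, loosest down the center, while anchor and body joints leave rotation free.
# Capsules also cover head, neck, upper spine and only the right clavicle.
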PhysicsAsset /BRCosmetics/Characters/Player/Female/Medium/Bodies/F_MED_Musician/Meshes/Parts/F_MED_Musician_FaceFE_Physics
bodies {
 clavicle_r [{shapeType=Capsule, length=19.375, radius=5.005, shapeName=clavicle_r_capsule, center=<-8.3578 0.6994 0>, rotation=<rot 87.7969 180 180>}]
 dyn_hair_01 [{shapeType=Capsule, length=6, radius=4, shapeName=dyn_hair_01_capsule, center=<4 0 0>, rotation=<rot 0 0 85>}]
 dyn_hair_01_l [{shapeType=Capsule, length=1.625, radius=2.75, shapeName=dyn_hair_01_l_capsule, center=<3 0 0>, rotation=<rot 0 0 90>}]
 dyn_hair_01_r [{shapeType=Capsule, length=1.875, radius=2.75, center=<3 0 0>, rotation=<rot 0 0 90>}]
 dyn_hair_02 [{
   shapeType=Capsule
   length=14.8125
   radius=3.187
   shapeName=dyn_hair_02_capsule
   center=<3.9891 -0.0453 -0.5173>
   rotation=<rot 0.0001 -0.0005 97.8126>
  }]
 dyn_hair_02_l [{shapeType=Capsule, length=2.5, radius=2.5, shapeName=dyn_hair_02_l_capsule, center=<3 0 0>, rotation=<rot 0 0 90>}]
 dyn_hair_02_r [{shapeType=Capsule, length=2.375, radius=2.625, shapeName=dyn_hair_02_r_capsule, center=<3 0 0>, rotation=<rot 0 0 90>}]
 dyn_hair_03 [{shapeType=Capsule, length=12.6875, radius=2.8125, shapeName=dyn_hair_03_capsule, center=<3.3599 0 0>, rotation=<rot 0 0 97.8125>}]
 dyn_hair_03_l [{shapeType=Capsule, length=1.625, radius=2.125, shapeName=dyn_hair_03_l_capsule, center=<3 0.7252 0>, rotation=<rot 0.0024 -0.0035 90.0021>}]
 dyn_hair_03_r [{shapeType=Capsule, length=2.875, radius=2.375, shapeName=dyn_hair_03_r_capsule, center=<3 0 0>, rotation=<rot 0 0 90>}]
 dyn_hair_04_l [{shapeType=Capsule, length=3, radius=2, shapeName=dyn_hair_04_l_capsule, center=<3.9648 0 0>, rotation=<rot 0.0001 89.9997 89.9999>}]
 dyn_hair_04_r [{shapeType=Capsule, length=2.5, radius=2, shapeName=dyn_hair_04_r_capsule, center=<3 0 0.7575>, rotation=<rot 0 90 90>}]
 dyn_hair_MSTR [{shapeType=Sphere, radius=0.505, shapeName=dyn_hair_MSTR_sphere}]
 dyn_hair_MSTR_l [{shapeType=Sphere, radius=0.505, shapeName=dyn_hair_MSTR_l_sphere}]
 dyn_hair_MSTR_r [{shapeType=Sphere, radius=0.505, shapeName=dyn_hair_MSTR_r_sphere}]
 dyn_hair_br [{shapeType=Capsule, length=3.5, radius=2.375, shapeName=dyn_hair_br_capsule, center=<4.3998 0 -0.2245>, rotation=<rot 90 0 0>}]
 dyn_hair_br_MSTR [{shapeType=Capsule, length=0.505, radius=0.505, shapeName=dyn_hair_br_MSTR_capsule, rotation=<rot 0 -90 0>}]
 head [{shapeType=Capsule, length=5.3125, radius=5.75, center=<0.078 0.8717 0>, rotation=<rot 90 0 0>}]
 neck_01 [{shapeType=Capsule, length=1.8125, radius=6.1925, shapeName=neck_01_capsule, center=<0 1.9063 0.0001>, rotation=<rot 0 -90 90>}]
 spine_04 [
  {shapeType=Capsule, length=26.625, radius=9.125, shapeName=spine_04_capsule, center=<5.5406 -0.2985 0>, rotation=<rot 0 -90 0>}
  {shapeType=Capsule, length=28.375, radius=7.375, shapeName=spine_04_capsule, center=<3.8459 2.4961 0>, rotation=<rot 0 -90 0>}
 ]
 spine_05 [{shapeType=Capsule, length=29.0625, radius=5.1875, shapeName=spine_05_capsule, center=<5.2483 -1.6989 -0.0002>, rotation=<rot 0 -90 0>}]
}
constraints [
 {bone1Name=dyn_hair_01_r, bone2Name=dyn_hair_MSTR_r, swing1LimitDegrees=10, swing2LimitDegrees=10, twistLimitDegrees=5}
 {bone1Name=dyn_hair_01, bone2Name=dyn_hair_MSTR, swing1LimitDegrees=20, swing2LimitDegrees=20, twistLimitDegrees=20}
 {bone1Name=dyn_hair_02, bone2Name=dyn_hair_01, swing1LimitDegrees=45, swing2LimitDegrees=55, twistLimitDegrees=20}
 {bone1Name=dyn_hair_01_l, bone2Name=dyn_hair_MSTR_l, swing1LimitDegrees=15, swing2LimitDegrees=15, twistLimitDegrees=5}
 {bone1Name=dyn_hair_02_l, bone2Name=dyn_hair_01_l, swing1LimitDegrees=20, swing2LimitDegrees=15, twistLimitDegrees=5}
 {bone1Name=dyn_hair_03_l, bone2Name=dyn_hair_02_l, swing1LimitDegrees=20, swing2LimitDegrees=15, twistLimitDegrees=5}
 {bone1Name=dyn_hair_04_l, bone2Name=dyn_hair_03_l, swing1LimitDegrees=20, swing2LimitDegrees=20, twistLimitDegrees=5}
 {bone1Name=dyn_hair_02_r, bone2Name=dyn_hair_01_r, swing1LimitDegrees=15, swing2LimitDegrees=15, twistLimitDegrees=5}
 {bone1Name=dyn_hair_03_r, bone2Name=dyn_hair_02_r, swing1LimitDegrees=15, swing2LimitDegrees=15, twistLimitDegrees=5}
 {bone1Name=dyn_hair_04_r, bone2Name=dyn_hair_03_r, swing1LimitDegrees=15, swing2LimitDegrees=25, twistLimitDegrees=5}
 {bone1Name=dyn_hair_03, bone2Name=dyn_hair_02, swing1LimitDegrees=45, swing2LimitDegrees=55, twistLimitDegrees=20}
 {
  bone1Name=head
  bone2Name=neck_01
  swing1LimitDegrees=45
  swing1Motion=Free
  swing2LimitDegrees=45
  swing2Motion=Free
  twistLimitDegrees=45
  twistMotion=Free
 }
 {
  bone1Name=spine_05
  bone2Name=spine_04
  swing1LimitDegrees=45
  swing1Motion=Free
  swing2LimitDegrees=45
  swing2Motion=Free
  twistLimitDegrees=45
  twistMotion=Free
 }
 {
  bone1Name=dyn_hair_MSTR
  bone2Name=head
  swing1LimitDegrees=45
  swing1Motion=Free
  swing2LimitDegrees=45
  swing2Motion=Free
  twistLimitDegrees=45
  twistMotion=Free
 }
 {
  bone1Name=dyn_hair_MSTR_r
  bone2Name=head
  swing1LimitDegrees=45
  swing1Motion=Free
  swing2LimitDegrees=45
  swing2Motion=Free
  twistLimitDegrees=45
  twistMotion=Free
 }
 {
  bone1Name=dyn_hair_MSTR_l
  bone2Name=head
  swing1LimitDegrees=45
  swing1Motion=Free
  swing2LimitDegrees=45
  swing2Motion=Free
  twistLimitDegrees=45
  twistMotion=Free
 }
 {
  bone1Name=clavicle_r
  bone2Name=spine_05
  swing1LimitDegrees=45
  swing1Motion=Free
  swing2LimitDegrees=45
  swing2Motion=Free
  twistLimitDegrees=45
  twistMotion=Free
 }
 {bone1Name=dyn_hair_br, bone2Name=dyn_hair_br_MSTR, swing1LimitDegrees=35, swing2LimitDegrees=35, twistLimitDegrees=20}
 {
  bone1Name=dyn_hair_br_MSTR
  bone2Name=dyn_hair_MSTR
  swing1LimitDegrees=45
  swing1Motion=Free
  swing2LimitDegrees=45
  swing2Motion=Free
  twistLimitDegrees=45
  twistMotion=Free
 }
 {
  bone1Name=neck_01
  bone2Name=spine_05
  swing1LimitDegrees=45
  swing1Motion=Free
  swing2LimitDegrees=45
  swing2Motion=Free
  twistLimitDegrees=45
  twistMotion=Free
 }
]
