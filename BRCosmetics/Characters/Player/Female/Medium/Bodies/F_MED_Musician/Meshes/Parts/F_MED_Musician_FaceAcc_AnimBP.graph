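# Post-process style AnimBlueprint for the female Musician face accessory, animating hair, hat and glasses over the parent's copied pose.
# Update logic maps head and upper-arm bone deltas to alphas for additive corrective bone offsets, cached before physics.
# RigidBody nodes with gravity override pick a physics asset by backbling, then add front-hair simulation and scale tweaks.
# The front end uses its own simpler RigidBody path; several corrective chains are unconnected.
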AnimBlueprint /BRCosmetics/Characters/Player/Female/Medium/Bodies/F_MED_Musician/Meshes/Parts/F_MED_Musician_FaceAcc_AnimBP
target_skeleton=&/F_MED_Musician_FaceAcc_Skeleton
graphs [
 AnimGraph {nodes [
   Layered blend per bone(BlendPoses_0=Copy Pose From Mesh(SourceMeshComponent=Get Skel() -> (Skel:Skeletal Mesh Component Object Reference)) {bCopyCurves=true, bUseAttachedParent=true}) {layerSetup [{branchFilters [{boneName=L_cheek_inner}]}]}
   n1 = "Transform (Modify) Bone - Bone: dyn_hair_MSTR_r"(Alpha=Get R Arm Rot() -> (R Arm Rot:double), Translation="0,0,-2", Rotation="-5,0,0", Scale="1,1,1") {
    rotationMode=BMM_Additive
    rotationSpace=BCS_BoneSpace
    translationMode=BMM_Additive
    translationSpace=BCS_BoneSpace
    alphaBoolBlend {blendOption=Linear}
    boneToModify {boneName=dyn_hair_MSTR_r}
    rotation {pitch=-5}
    translation {z=-2}
   }
   "Transform (Modify) Bone - Bone: dyn_hair_MSTR_l"(ComponentPose=@n1, Alpha=Get L Arm Rot() -> (L Arm Rot:double), Translation="0,0,-2", Rotation="10,0,0", Scale="1,1,1") {
    rotationMode=BMM_Additive
    rotationSpace=BCS_BoneSpace
    translationMode=BMM_Additive
    translationSpace=BCS_BoneSpace
    alphaBoolBlend {blendOption=Linear}
    boneToModify {boneName=dyn_hair_MSTR_l}
    rotation {pitch=10}
    translation {z=-2}
   }
   n2 = Get Head Side Rot Alpha() -> (Head Side Rot Alpha:double)
   "Transform (Modify) Bone - Bone: dyn_hair_br"(Alpha=@n2, Translation="0,5,5", Rotation="-20,0,0", Scale="1,1,1") {
    rotationMode=BMM_Additive
    rotationSpace=BCS_BoneSpace
    translationMode=BMM_Additive
    translationSpace=BCS_BoneSpace
    alphaBoolBlend {blendOption=Linear}
    boneToModify {boneName=dyn_hair_br}
    rotation {pitch=-20}
    translation {y=5, z=5}
   }
   "Transform (Modify) Bone - Bone: dyn_hair_01_r"(Alpha=1, Translation="0,0,0", Rotation="0,0,0", Scale="1,0.5,0.5") {
    rotationSpace=BCS_BoneSpace
    scaleMode=BMM_Replace
    translationMode=BMM_Additive
    translationSpace=BCS_BoneSpace
    alphaBoolBlend {blendOption=Linear}
    boneToModify {boneName=dyn_hair_01_r}
    scale {y=0.5, z=0.5}
   }
   "Transform (Modify) Bone - Bone: dyn_hair_br_MSTR"(Alpha=Get R Hair Rot() -> (R Hair Rot:double), Translation="0,0,15", Rotation="0,0,0", Scale="1,1,1") {
    rotationMode=BMM_Additive
    rotationSpace=BCS_BoneSpace
    translationMode=BMM_Additive
    translationSpace=BCS_BoneSpace
    alphaBoolBlend {blendOption=Linear}
    boneToModify {boneName=dyn_hair_br_MSTR}
    translation {z=15}
   }
   n3 = Get Head Rot Alpha() -> (Head Rot Alpha:double)
   n4 = "Transform (Modify) Bone - Bone: dyn_hair_MSTR"(Alpha=@n3, Translation="-3,2,2", Rotation="0,0,0", Scale="1,1,1") {
    rotationMode=BMM_Additive
    rotationSpace=BCS_BoneSpace
    translationMode=BMM_Additive
    translationSpace=BCS_BoneSpace
    alphaBoolBlend {blendOption=Linear}
    boneToModify {boneName=dyn_hair_MSTR}
    translation=<-3 2 2>
   }
   n5 = "Transform (Modify) Bone - Bone: dyn_hair_MSTR_l"(ComponentPose=@n4, Alpha=@n3, Translation="0,-2,0", Rotation="0,0,0", Scale="1,1,1") {
    rotationMode=BMM_Additive
    rotationSpace=BCS_BoneSpace
    translationMode=BMM_Additive
    translationSpace=BCS_BoneSpace
    alphaBoolBlend {blendOption=Linear}
    boneToModify {boneName=dyn_hair_MSTR_l}
    translation {y=-2}
   }
   n6 = "Transform (Modify) Bone - Bone: dyn_hat"(ComponentPose=@n5, Alpha=@n3, Translation="0,0,1", Rotation="0,0,7", Scale="1,1,1") {
    rotationMode=BMM_Additive
    rotationSpace=BCS_BoneSpace
    translationMode=BMM_Additive
    translationSpace=BCS_BoneSpace
    alphaBoolBlend {blendOption=Linear}
    boneToModify {boneName=dyn_hat}
    rotation {roll=7}
    translation {z=1}
   }
   n7 = Get Head Side Rot Alpha()
   n8 = "Transform (Modify) Bone - Bone: dyn_hair_MSTR"(ComponentPose=@n6, Alpha=@n7, Translation="0,0,3", Rotation="0.000,-20,30", Scale="1,1,1") {
    rotationMode=BMM_Additive
    rotationSpace=BCS_BoneSpace
    translationMode=BMM_Additive
    translationSpace=BCS_BoneSpace
    alphaBoolBlend {blendOption=Linear}
    boneToModify {boneName=dyn_hair_MSTR}
    rotation {roll=30, yaw=-20}
    translation {z=3}
   }
   n9 = "Transform (Modify) Bone - Bone: dyn_hat"(ComponentPose=@n8, Alpha=@n7, Translation="0,0,0.6", Rotation="0,0,3", Scale="1,1,1") {
    rotationMode=BMM_Additive
    rotationSpace=BCS_BoneSpace
    translationMode=BMM_Additive
    translationSpace=BCS_BoneSpace
    alphaBoolBlend {blendOption=Linear}
    boneToModify {boneName=dyn_hat}
    rotation {roll=3}
    translation {z=0.6}
   }
   n10 = "Transform (Modify) Bone - Bone: dyn_hair_bl"(ComponentPose=@n9, Alpha=Get Head Rot Alpha(), Translation="0,0,5", Rotation="0,0,0", Scale="1,1,1") {
    rotationMode=BMM_Additive
    rotationSpace=BCS_BoneSpace
    translationMode=BMM_Additive
    translationSpace=BCS_BoneSpace
    alphaBoolBlend {blendOption=Linear}
    boneToModify {boneName=dyn_hair_bl}
    translation {z=5}
   }
   n11 = "Transform (Modify) Bone - Bone: dyn_hair_bl"(ComponentPose=@n10, Alpha=Get Head Side Rot Alpha(), Translation="0,-2,5", Rotation="0,0,0", Scale="1,1,1") {
    rotationMode=BMM_Additive
    rotationSpace=BCS_BoneSpace
    translationMode=BMM_Additive
    translationSpace=BCS_BoneSpace
    alphaBoolBlend {blendOption=Linear}
    boneToModify {boneName=dyn_hair_bl}
    translation {y=-2, z=5}
   }
   n12 = Get Head Side Rot Alpha()
   n13 = "Transform (Modify) Bone - Bone: dyn_hair_03_r1"(ComponentPose=@n11, Alpha=float / float(A=@n12, B=2) {is=title K2Node_CallFunction}, Translation="0,-8,0", Rotation="0,0,0", Scale="1,1,1") {
    rotationMode=BMM_Additive
    rotationSpace=BCS_BoneSpace
    translationMode=BMM_Additive
    translationSpace=BCS_BoneSpace
    alphaBoolBlend {blendOption=Linear}
    boneToModify {boneName=dyn_hair_03_r1}
    translation {y=-8}
   }
   n14 = Get Head Rot Alpha()
   n15 = "Transform (Modify) Bone - Bone: dyn_hair_03_r2"(ComponentPose=@n13, Alpha=@n14, Translation="0,2,0", Rotation="0,0,0", Scale="1,1,1") {
    rotationMode=BMM_Additive
    rotationSpace=BCS_BoneSpace
    translationMode=BMM_Additive
    translationSpace=BCS_BoneSpace
    alphaBoolBlend {blendOption=Linear}
    boneToModify {boneName=dyn_hair_03_r2}
    translation {y=2}
   }
   n16 = "Transform (Modify) Bone - Bone: dyn_hair_03_l1"(ComponentPose=@n15, Alpha=@n12, Translation="0,10,10", Rotation="0,0,0", Scale="1,1,1") {
    rotationMode=BMM_Additive
    rotationSpace=BCS_BoneSpace
    translationMode=BMM_Additive
    translationSpace=BCS_BoneSpace
    alphaBoolBlend {blendOption=Linear}
    boneToModify {boneName=dyn_hair_03_l1}
    translation {y=10, z=10}
   }
   "Transform (Modify) Bone - Bone: dyn_hair_03_l2"(ComponentPose=@n16, Alpha=@n14, Translation="0,-10,0", Rotation="0,0,0", Scale="1,1,1") {
    rotationMode=BMM_Additive
    rotationSpace=BCS_BoneSpace
    translationMode=BMM_Additive
    translationSpace=BCS_BoneSpace
    alphaBoolBlend {blendOption=Linear}
    boneToModify {boneName=dyn_hair_03_l2}
    translation {y=-10}
   }
   "Transform (Modify) Bone - Bone: dyn_hair_br_MSTR"(Alpha=@n2, Translation="0,0,15", Rotation="0,0,0", Scale="1,1,1") {
    rotationMode=BMM_Additive
    rotationSpace=BCS_BoneSpace
    translationMode=BMM_Additive
    translationSpace=BCS_BoneSpace
    alphaBoolBlend {blendOption=Linear}
    boneToModify {boneName=dyn_hair_br_MSTR}
    translation {z=15}
   }
   n17 = RigidBody!(
    ComponentPose=Local To Component(LocalPose=Copy Pose From Mesh() {bCopyCurves=true, bUseAttachedParent=true})
    Alpha=0.7
    OverrideWorldGravity:Vector=Get Gravity Override()
    ExternalForce
   ) -> (Pose) {
    bClampLinearTranslationLimitToRefPose=true
    bOverrideWorldGravity=true
    overridePhysicsAsset=&/F_MED_Musician_FaceFE_Physics
    simulationSpace=BaseBoneSpace
    alphaBoolBlend {blendOption=Linear}
    baseBoneRef {boneName=neck_01}
    componentAppliedLinearAccClamp=<3000 3000 3000>
    componentLinearAccScale=<0.2 0.2 0.3>
    componentLinearVelScale=<0.2 0.2 0.3>
    overrideWorldGravity {z=-980}
   }
   n18 = RigidBody(ComponentPose=Local To Component(LocalPose=Use cached pose 'preBackpackCheck'()), OverrideWorldGravity=Get Gravity Override()) {
    bClampLinearTranslationLimitToRefPose=true
    bOverrideWorldGravity=true
    overridePhysicsAsset=&/F_MED_Musician_FaceWBack_Physics
    simulationSpace=BaseBoneSpace
    alphaBoolBlend {blendOption=Linear}
    baseBoneRef {boneName=neck_01}
    componentAppliedLinearAccClamp=<3000 3000 3000>
    componentLinearAccScale=<0.2 0.2 0.3>
    componentLinearVelScale=<0.2 0.2 0.3>
    overrideWorldGravity {z=-980}
   }
   n19 = Component To Local(ComponentPose=@n18)
   n20 = RigidBody!(
    ComponentPose=Local To Component(LocalPose=Use cached pose 'preBackpackCheck'())
    bAlphaBoolEnabled:bool
    OverrideWorldGravity=Get Gravity Override()
    ExternalForce
   ) -> (Pose) {
    alphaInputType=Bool
    bClampLinearTranslationLimitToRefPose=true
    bOverrideWorldGravity=true
    overridePhysicsAsset=&/F_MED_Musician_FaceAcc_Physics
    simulationSpace=BaseBoneSpace
    alphaBoolBlend {blendOption=Linear}
    baseBoneRef {boneName=neck_01}
    componentAppliedLinearAccClamp=<3000 3000 3000>
    componentLinearAccScale=<0.4 0.4 0.6>
    componentLinearVelScale=<0.2 0.2 0.3>
    overrideWorldGravity {z=-980}
   }
   n21 = Component To Local(ComponentPose=@n20)
   n22 = Get Is Wearing Backbling()
   n23 = RigidBody(
    ComponentPose=Local To Component(LocalPose=Blend Poses by bool(BlendPose_0=@n19, BlendPose_1=@n21, bActiveValue=@n22) {blendType=Linear})
    OverrideWorldGravity=Get Gravity Override()
   ) {
    alphaInputType=Bool
    bClampLinearTranslationLimitToRefPose=true
    bOverrideWorldGravity=true
    overridePhysicsAsset=&/F_MED_Musician_FaceAcc_FrontHair_Physics
    simulationSpace=BaseBoneSpace
    alphaBoolBlend {blendOption=Linear}
    baseBoneRef {boneName=neck_01}
    componentAppliedLinearAccClamp=<1000 1000 1000>
    componentLinearAccScale=<0.25 0.25 0.25>
    componentLinearVelScale=<0.15 0.15 0.2>
    overrideWorldGravity {z=-980}
   }
   n24 = "Transform (Modify) Bone - Bone: dyn_hair_02_r"(ComponentPose=@n23, Alpha=1, Translation="0,0,0", Rotation="0,0,0", Scale="1,0.5,0.5") {
    rotationSpace=BCS_BoneSpace
    scaleMode=BMM_Replace
    translationMode=BMM_Additive
    translationSpace=BCS_BoneSpace
    alphaBoolBlend {blendOption=Linear}
    boneToModify {boneName=dyn_hair_02_r}
    scale {y=0.5, z=0.5}
   }
   n25 = "Transform (Modify) Bone - Bone: dyn_hair_03_r"(ComponentPose=@n24, Alpha=1, Translation="2,0,2", Rotation="0,0,0", Scale="1,0.8,0.5") {
    rotationSpace=BCS_BoneSpace
    scaleMode=BMM_Replace
    translationMode=BMM_Additive
    translationSpace=BCS_BoneSpace
    alphaBoolBlend {blendOption=Linear}
    boneToModify {boneName=dyn_hair_03_r}
    scale {y=0.8, z=0.5}
    translation {x=2, z=2}
   }
   Output Pose(Result=Blend Poses by bool(BlendPose_0=Component To Local(ComponentPose=@n17), BlendPose_1=Component To Local(ComponentPose=@n25), bActiveValue=Get bIsInFrontEnd()) {blendType=Linear}) {layerGroup=None}
   n26 = "Transform (Modify) Bone - Bone: upperarm_l"!(ComponentPose=Local To Component(LocalPose=Copy Pose From Mesh() {bCopyCurves=true, bUseAttachedParent=true}), Alpha=1, Rotation="0,0,0") -> (Pose) {rotationMode=BMM_Replace, rotationSpace=BCS_ParentBoneSpace, alphaBoolBlend {blendOption=Linear}, boneToModify {boneName=upperarm_l}}
   n27 = "Transform (Modify) Bone - Bone: upperarm_r"!(ComponentPose=@n26, Alpha=1, Rotation="0,0,0") -> (Pose) {rotationMode=BMM_Replace, rotationSpace=BCS_ParentBoneSpace, alphaBoolBlend {blendOption=Linear}, boneToModify {boneName=upperarm_r}}
   n28 = "Transform (Modify) Bone - Bone: dyn_hair_MSTR"(ComponentPose=@n27, Alpha=1, Translation="-2.5,0,0", Rotation="0,0,0", Scale="1,1,1") {
    rotationMode=BMM_Additive
    rotationSpace=BCS_BoneSpace
    translationMode=BMM_Additive
    translationSpace=BCS_BoneSpace
    alphaBoolBlend {blendOption=Linear}
    boneToModify {boneName=dyn_hair_MSTR}
    translation {x=-2.5}
   }
   n29 = "Transform (Modify) Bone - Bone: dyn_hair_MSTR"(ComponentPose=@n28, Alpha=Get Head Rot Alpha(), Translation="0,0,1", Rotation="30,0.000,0.000", Scale="1,1,1") {
    rotationMode=BMM_Additive
    rotationSpace=BCS_BoneSpace
    translationMode=BMM_Additive
    translationSpace=BCS_BoneSpace
    alphaBoolBlend {blendOption=Linear}
    boneToModify {boneName=dyn_hair_MSTR}
    rotation {pitch=30}
    translation {z=1}
   }
   n30 = "Transform (Modify) Bone - Bone: dyn_hat"(ComponentPose=@n29, Alpha=Get Head Rot Alpha(), Translation="0,0,1", Rotation="0,0,7", Scale="1,1,1") {
    rotationMode=BMM_Additive
    rotationSpace=BCS_BoneSpace
    translationMode=BMM_Additive
    translationSpace=BCS_BoneSpace
    alphaBoolBlend {blendOption=Linear}
    boneToModify {boneName=dyn_hat}
    rotation {roll=7}
    translation {z=1}
   }
   n31 = "Transform (Modify) Bone - Bone: dyn_hat"(ComponentPose=@n30, Alpha=Get Head Side Rot Alpha(), Translation="0,0,0.6", Rotation="0,0,4", Scale="1,1,1") {
    rotationMode=BMM_Additive
    rotationSpace=BCS_BoneSpace
    translationMode=BMM_Additive
    translationSpace=BCS_BoneSpace
    alphaBoolBlend {blendOption=Linear}
    boneToModify {boneName=dyn_hat}
    rotation {roll=4}
    translation {z=0.6}
   }
   n32 = "Transform (Modify) Bone - Bone: dyn_hair_MSTR_l"(ComponentPose=@n31, Alpha=Get Head Rot Alpha(), Translation="0,-2,0", Rotation="0,0,0", Scale="1,1,1") {
    rotationMode=BMM_Additive
    rotationSpace=BCS_BoneSpace
    translationMode=BMM_Additive
    translationSpace=BCS_BoneSpace
    alphaBoolBlend {blendOption=Linear}
    boneToModify {boneName=dyn_hair_MSTR_l}
    translation {y=-2}
   }
   n33 = Get Side hair alpha() -> (Side hair alpha:double)
   n34 = "Transform (Modify) Bone - Bone: dyn_hair_MSTR_l"(ComponentPose=@n32, Alpha=@n33, Translation="0,0,0", Rotation="10,0,0", Scale="1,1,1") {
    rotationMode=BMM_Additive
    rotationSpace=BCS_BoneSpace
    translationMode=BMM_Additive
    translationSpace=BCS_BoneSpace
    alphaBoolBlend {blendOption=Linear}
    boneToModify {boneName=dyn_hair_MSTR_l}
    rotation {pitch=10}
   }
   n35 = "Transform (Modify) Bone - Bone: dyn_hair_MSTR_l"(ComponentPose=@n34, Alpha=@n33, Translation="-1,0,0", Rotation="0,5,0", Scale="1,1,1") {
    rotationMode=BMM_Additive
    rotationSpace=BCS_BoneSpace
    translationMode=BMM_Additive
    translationSpace=BCS_BoneSpace
    alphaBoolBlend {blendOption=Linear}
    boneToModify {boneName=dyn_hair_MSTR_l}
    rotation {yaw=5}
    translation {x=-1}
   }
   n36 = "Transform (Modify) Bone - Bone: dyn_hair_bl"(ComponentPose=@n35, Alpha=Get Head Rot Alpha(), Translation="0,0,0", Rotation="0,0,0", Scale="1,1,1") {
    rotationMode=BMM_Additive
    rotationSpace=BCS_BoneSpace
    translationMode=BMM_Additive
    translationSpace=BCS_BoneSpace
    alphaBoolBlend {blendOption=Linear}
    boneToModify {boneName=dyn_hair_bl}
   }
   n37 = "Transform (Modify) Bone - Bone: dyn_hair_bl"(ComponentPose=@n36, Alpha=Get Head Side Rot Alpha(), Translation="0,0,0", Rotation="0,0,0", Scale="1,1,1") {
    rotationMode=BMM_Additive
    rotationSpace=BCS_BoneSpace
    translationMode=BMM_Additive
    translationSpace=BCS_BoneSpace
    alphaBoolBlend {blendOption=Linear}
    boneToModify {boneName=dyn_hair_bl}
   }
   n38 = "Transform (Modify) Bone - Bone: dyn_hair_br_MSTR"(ComponentPose=@n37, Alpha=Get Head Side Rot Alpha(), Translation="-9,4,9", Rotation="30,20,0", Scale="1,1,1") {
    rotationMode=BMM_Additive
    rotationSpace=BCS_BoneSpace
    translationMode=BMM_Additive
    translationSpace=BCS_BoneSpace
    alphaBoolBlend {blendOption=Linear}
    boneToModify {boneName=dyn_hair_br_MSTR}
    rotation {pitch=30, yaw=20}
    translation=<-9 4 9>
   }
   n39 = "Transform (Modify) Bone - Bone: dyn_hair_bl"(ComponentPose=@n38, Alpha=Get L Arm Rot(), Translation="-3,0,3", Rotation="0.000,0.000,35", Scale="1,1,1") {
    rotationMode=BMM_Additive
    rotationSpace=BCS_BoneSpace
    translationMode=BMM_Additive
    translationSpace=BCS_BoneSpace
    alphaBoolBlend {blendOption=Linear}
    boneToModify {boneName=dyn_hair_bl}
    rotation {roll=35}
    translation {x=-3, z=3}
   }
   n40 = "Transform (Modify) Bone - Bone: dyn_glasses"(ComponentPose=@n39, Alpha=1, Translation="0,0.1,0", Rotation="0,0,0", Scale="1,1,1") {
    translationMode=BMM_Additive
    translationSpace=BCS_BoneSpace
    alphaBoolBlend {blendOption=Linear}
    boneToModify {boneName=dyn_glasses}
    translation {y=0.1}
   }
   Save cached pose 'preBackpackCheck'(Pose=Component To Local(ComponentPose=@n40)) {cacheName=preBackpackCheck}
   Use cached pose 'preBackpackCheck'()
   Get Head Rot Alpha()
  ]}
 EventGraph {nodes [
   Event BlueprintUpdateAnimation()
   > Set Skel(Skel:Skeletal Mesh Component Object Reference=GetOwnerSkeletalMeshComponent()) -> (Output_Get:Skeletal Mesh Component Object Reference)
   > n1 = Sequence!(execute) -> (then_0, then_1, then_2:Exec)
   > n2 = Gravity Override State Function DEPRECATED(CharacterPartAnimInstance=Self-Reference(), GameTimeInSeconds=GetGameTimeInSeconds(), GravityOverrideParams=Get Gravity Override Params())
   > Set Gravity Override(Gravity Override=@n2)
   Set Rotation_Upperarm_L(
    execute=@n1.then_2
    Rotation_Upperarm_L:Rotator=GetBoneTransform!(
     self:Skinned Mesh Component Object Reference=Get BodyMesh() -> (BodyMesh:Skeletal Mesh Component Budgeted Object Reference)
     InBoneName:Name=upperarm_l
     TransformSpace:ERelativeTransformSpace Enum=RTS_ParentBoneSpace
    ) -> (ReturnValue_Location:Vector, ReturnValue_Rotation:Rotator, ReturnValue_Scale:Vector).ReturnValue_Rotation
   ) -> (Output_Get:Rotator)
   > Set Rotation_Upperarm_R(Rotation_Upperarm_R:Rotator=GetBoneTransform(self=Get BodyMesh(), InBoneName=upperarm_r, TransformSpace=RTS_ParentBoneSpace).ReturnValue_Rotation) -> (Output_Get:Rotator)
   n3 = GetOwnerSkeletalMeshComponent()
   n4 = BreakRotator(InRot=BreakTransform(InTransform=GetDeltaTransformFromRefPose(self=@n3, BoneName=head, BaseName=spine_05)).Rotation)
   > Set Head_Back(Head_Back:double=SelectFloat(A=MapRangeClamped(Value=@n4.Yaw, InRangeA=-10, InRangeB=-55), B=0, bPickA=Get bIsCrouching())) -> (Output_Get:double)
   > Set Head_Right(Head_Right:double=MapRangeClamped(Value=@n4.Pitch, InRangeA=0, InRangeB=45)) -> (Output_Get:double)
   n5 = GetOwnerSkeletalMeshComponent()
   n6 = BreakRotator(InRot=BreakTransform(InTransform=GetDeltaTransformFromRefPose(self=@n5, BoneName=head, BaseName=spine_05)).Rotation)
   Set Head Rot Alpha(execute=@n1.then_1, Head Rot Alpha:double=MapRangeClamped(Value=@n6.Yaw, InRangeA=0, InRangeB=-55)) -> (Output_Get:double)
   > Set Head Side Rot Alpha(Head Side Rot Alpha:double=MapRangeClamped(Value=@n6.Pitch, InRangeA=-10, InRangeB=35)) -> (Output_Get:double)
   > Set Side hair alpha(Side hair alpha:double=MapRangeClamped(Value=@n6.Pitch, InRangeA=-10, InRangeB=0)) -> (Output_Get:double)
   n7 = GetOwnerSkeletalMeshComponent()
   n8 = BreakRotator(InRot=BreakTransform(InTransform=GetDeltaTransformFromRefPose(self=@n7, BoneName=upperarm_r, BaseName=spine_05)).Rotation)
   > Set R Arm Rot(R Arm Rot:double=MapRangeClamped(Value=@n8.Yaw, InRangeA=75, InRangeB=120)) -> (Output_Get:double)
   > Set R Hair Rot(R Hair Rot:double=MapRangeClamped(Value=@n8, InRangeA=25, InRangeB=50, OutRangeA=1, OutRangeB=0)) -> (Output_Get:double)
   n9 = GetDeltaTransformFromRefPose(self=GetOwnerSkeletalMeshComponent(), BoneName=L_cheek_inner, BaseName=head)
   > Set X(X:double=BreakVector(InVec=BreakTransform(InTransform=@n9))) -> (Output_Get:double)
   > n10 = Branch(Condition=Get bIsInFrontEnd())
   n11 = BreakTransform(InTransform=GetDeltaTransformFromRefPose(self=GetOwnerSkeletalMeshComponent(), BoneName=upperarm_l, BaseName=spine_05))
   > Set L Arm Rot(L Arm Rot:double=MapRangeClamped(Value=BreakRotator(InRot=@n11.Rotation).Yaw, InRangeA=0, InRangeB=-16)) -> (Output_Get:double)
   Set L Arm Rot(execute=@n10.else, L Arm Rot=0)
   n12 = Event OnCharacterCustomizationLoaded()
   > Set Is Wearing Backbling(Is Wearing Backbling=CheckForBackbling())
   > Set BodyMesh(BodyMesh:Skeletal Mesh Component Budgeted Object Reference=Get Array Item with EFort Custom Part Type(Array=@n12.CustomMeshParts)) -> (Output_Get:Skeletal Mesh Component Budgeted Object Reference)
   Get Skel()
  ]}
]
variables [
 BodyMesh {type=SkeletalMeshComponentBudgeted}
 Delta Time X {type=number}
 Gravity Override {type=Vector}
 Gravity Override Params {
  type=GravityOverrideParamsStruct
  default {skydive_DivingDown {gravityOverride=<0 0 0>, windFrequency=<1.5 7 8>}, skydive_GoingUp {gravityOverride=<0 0 -100>}}
 }
 Head Rot Alpha {type=number}
 Head Side Rot Alpha {type=number}
 Head_Back {type=number}
 Head_Right {type=number}
 Is Wearing Backbling {type=boolean}
 L Arm Rot {type=number}
 Location_Upperarm_L {type=Vector}
 Location_Upperarm_R {type=Vector}
 NewVar {type=number}
 R Arm Rot {type=number}
 R Hair Rot {type=number}
 Rotation_Upperarm_L {type=Rotator}
 Rotation_Upperarm_R {type=Rotator}
 Side hair alpha {type=number}
 Skel {type=SkeletalMeshComponent}
 Transform_Upperarm_L {type=Transform, default {location=<0 0 0>, rotation=<rot 0 0 0>, scale=<1 1 1>}}
 Transform_Upperarm_R {type=Transform, default {location=<0 0 0>, rotation=<rot 0 0 0>, scale=<1 1 1>}}
 Variable {type=number}
 X {type=number}
]
